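# Material function for Modular Button Base materials that blends Idle, Hovered, Focused, Pressed and Disabled visuals into Color and Opacity outputs.
# Each state input defaults to the RGB or alpha of its exposed color parameter, so a material can override any of them.
# Chained lerps driven by shared Trans_* scalar parameters layer the states in that order, so Disabled overrides Pressed, then Focused, then Hovered.
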MaterialFunction /FMJamSystem/Widgets/Materials/MF_UI_JamButtonStatesBase_Lerp
graph {
 properties {
  bExposeToLibrary=true
  description=Allows defining visuals for Modular Button Base states inside the material. Controls switching between the states by using transition parameters.
  libraryCategoriesText [UI Modular Button]
 }
 nodes [
  n1 = VectorParameter() {group=State colors, parameterName=Color_Idle, sortPriority=0, defaultValue=<rgba 1 1 1 1>}
  n2 = FunctionInput(Preview=@n1.RGB) {bUsePreviewValueAsDefault=true, description=Color for Idle state, inputName=Idle, previewValue=<1 1 1 1>}
  n3 = VectorParameter() {group=State colors, parameterName=Color_Hovered, sortPriority=1, defaultValue=<rgba 1 1 1 1>}
  n4 = FunctionInput(Preview=@n3.RGB) {bUsePreviewValueAsDefault=true, description=Color for Hovered state, inputName=Hovered, sortPriority=10, previewValue=<1 1 1 1>}
  n5 = ScalarParameter() {group=State transitions, parameterName=Trans_Hovered, sliderMax=1, sortPriority=1}
  n6 = VectorParameter() {group=State colors, parameterName=Color_Focused, sortPriority=2, defaultValue=<rgba 1 1 1 1>}
  n7 = FunctionInput(Preview=@n6.RGB) {bUsePreviewValueAsDefault=true, description=Color for Focused state, inputName=Focused, sortPriority=20, previewValue=<1 1 1 1>}
  n8 = ScalarParameter() {group=State transitions, parameterName=Trans_Focused, sliderMax=1, sortPriority=2}
  n9 = LinearInterpolate(A=LinearInterpolate(A=@n2, B=@n4, Alpha=@n5), B=@n7, Alpha=@n8)
  n10 = VectorParameter() {group=State colors, parameterName=Color_Pressed, sortPriority=3, defaultValue=<rgba 1 1 1 1>}
  n11 = FunctionInput(Preview=@n10.RGB) {bUsePreviewValueAsDefault=true, description=Color for Pressed state, inputName=Pressed, sortPriority=40}
  n12 = ScalarParameter() {group=State transitions, parameterName=Trans_Pressed, sliderMax=1, sortPriority=3}
  n13 = VectorParameter() {group=State colors, parameterName=Color_Disabled, sortPriority=4, defaultValue=<rgba 1 1 1 1>}
  n14 = ScalarParameter() {group=State transitions, parameterName=Trans_Disabled, sliderMax=1, sortPriority=4}
  FunctionOutput(None=LinearInterpolate(
    A=LinearInterpolate(A=@n9, B=@n11, Alpha=@n12)
    B=FunctionInput(Preview=@n13.RGB) {bUsePreviewValueAsDefault=true, description=Color for Disabled state, inputName=Disabled, sortPriority=50}
    Alpha=@n14
   )) {outputName=Color}
  n15 = FunctionInput(Preview=@n1.A) {bUsePreviewValueAsDefault=true, description=Opacity for Idle state, inputName=Idle Opacity, inputType=FunctionInput_Scalar, sortPriority=1}
  n16 = FunctionInput(Preview=@n3.A) {
   bUsePreviewValueAsDefault=true
   description=Opacity for Hovered state
   inputName=Hovered Opacity
   inputType=FunctionInput_Scalar
   sortPriority=11
  }
  n17 = FunctionInput(Preview=@n6.A) {
   bUsePreviewValueAsDefault=true
   description=Opacity for Focused state
   inputName=Focused Opacity
   inputType=FunctionInput_Scalar
   sortPriority=21
  }
  n18 = LinearInterpolate(A=LinearInterpolate(A=@n15, B=@n16, Alpha=@n5), B=@n17, Alpha=@n8)
  n19 = FunctionInput(Preview=@n10.A) {
   bUsePreviewValueAsDefault=true
   description=Opacity for Pressed state
   inputName=Pressed Opacity
   inputType=FunctionInput_Scalar
   sortPriority=41
  }
  n20 = FunctionInput(Preview=@n13.A) {
   bUsePreviewValueAsDefault=true
   description=Opacity for Disabled state
   inputName=Disabled Opacity
   inputType=FunctionInput_Scalar
   sortPriority=51
  }
  FunctionOutput(None=LinearInterpolate(A=LinearInterpolate(A=@n18, B=@n19, Alpha=@n12), B=@n20, Alpha=@n14)) {outputName=Opacity, sortPriority=1}
 ]
}
