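# UI material function in UI_ShaderLibrary that segments UV space into four quadrant masks for the top-left, top-right, bottom-right and bottom-left corners.
# It rounds the UV input, previewed with GetUserInterfaceUV's normalized UV, into binary X and Y values.
# Branchless Max, Multiply and OneMinus math combines them: inverted max gives top-left, the product bottom-right, and their exclusive-or splits into the other two.
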MaterialFunction /FortUILibrary/MaterialFunctions/Basics/MF_UI_CornerMask
graph {
 properties {bExposeToLibrary=true, description=Segment the 4 quadrants, libraryCategoriesText [UI_ShaderLibrary]}
 nodes [
  n1 = Round(None=FunctionInput(Preview=MaterialFunctionCall() {materialFunction=&GetUserInterfaceUV}.Normalized UV) {description=Set the UV, inputName=UV, inputType=FunctionInput_Vector2})
  n2 = ComponentMask(None=@n1) {r=true}
  n3 = ComponentMask(None=@n1) {g=true}
  n4 = Max(A=@n2, B=@n3)
  FunctionOutput(None=OneMinus(None=@n4)) {description=Expose the top left corner, outputName=Top Left Mask, sortPriority=10}
  n5 = Multiply(A=@n2, B=@n3)
  n6 = Multiply(A=@n4, B=OneMinus(None=@n5))
  FunctionOutput(None=Multiply(A=@n2, B=@n6)) {description=Expose the top right corner, outputName=Top Right Mask, sortPriority=20}
  FunctionOutput(None=@n5) {description=Expose the bottom right corner, outputName=Bottom Right Mask, sortPriority=30}
  FunctionOutput(None=Multiply(A=@n6, B=@n3)) {description=Expose the bottom left corner, outputName=Bottom Left Mask, sortPriority=40}
 ]
}
