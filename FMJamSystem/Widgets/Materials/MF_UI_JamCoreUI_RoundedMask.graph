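# UI material function that generates a rounded-corner box shape following Core UI rules, outputting shape and outline masks, their SDFs, and UVs.
# Per-corner radius, as a ratio or in pixels, is clamped and scaled by the smaller size axis to drive MF_UI_Box, with a skew angle input for slanted UVs.
# MF_UI_ShapeOutline derives the outline, and MF_UI_AntiAliasing produces both masks.
# Manual resolution-based SmoothStep edge branches remain unconnected to outputs.
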MaterialFunction /FMJamSystem/Widgets/Materials/MF_UI_JamCoreUI_RoundedMask
graph {
 properties {bExposeToLibrary=true, description=Generate a rounded corner shape based on Core UI rules, libraryCategoriesText [UI Library]}
 nodes [
  n1 = NamedRerouteUsage()
  n2 = FunctionInput(Preview=Constant2Vector() {g=1, r=1}) {bUsePreviewValueAsDefault=true, description=Defines the size, inputName=Size, inputType=FunctionInput_Vector2, sortPriority=1}
  n3 = FunctionInput(Preview=Constant() {r=0.3}) {bUsePreviewValueAsDefault=true, description=Set the corners radius, inputName=Corner Radius, inputType=FunctionInput_Vector4, sortPriority=2}
  n4 = NamedRerouteDeclaration(None=@n3) {name=Corner Radius, nodeColor=<rgba 0 0 0 1>}
  n5 = Divide(A=@n4, B=NamedRerouteUsage())
  n6 = FunctionInput(Preview=Constant()) {
   bUsePreviewValueAsDefault=true
   description=Set if you desire to calculate in pixels instead of a ratio/percentage
   inputName=Radius in Pixels
   inputType=FunctionInput_Scalar
   sortPriority=3
  }
  n7 = Floor(None=@n6)
  n8 = Multiply(A=Multiply(A=LinearInterpolate(A=@n4, B=@n5, Alpha=@n7)) {constB=2}, B=LinearInterpolate(B=NamedRerouteUsage(), Alpha=@n7) {constA=1})
  n9 = NamedRerouteUsage()
  n10 = ComponentMask(None=@n9) {r=true}
  n11 = ComponentMask(None=@n9) {g=true}
  n12 = Multiply(
   A=NamedRerouteDeclaration(None=Min(A=@n8) {desc=Avoid rounded corners to exceed the maximum amount}) {name=Remapped Corner Radius, nodeColor=<rgba 0 0 0 1>}
   B=NamedRerouteDeclaration(None=Min(A=@n10, B=@n11)) {name=Min Size, nodeColor=<rgba 0 0 0 1>}
  )
  n13 = FunctionOutput(None=MaterialFunctionCall!(
    UVs=@n1
    Size=NamedRerouteDeclaration(None=@n2) {name=Size, nodeColor=<rgba 0 0 0 1>}
    CornerRoundness=NamedRerouteDeclaration(None=@n12) {name=Corner Radius with Padding, nodeColor=<rgba 0 0 0 1>}
    OutlineThickness
    OutlineAlignment
    GlowMin
    GlowMax
    Keep Aspect Ratio=StaticBool() {value=true}
   ) {materialFunction=&MF_UI_Box}.SDF) {description=Output the SDF, outputName=SDF, sortPriority=10}
  FunctionOutput(None=MaterialFunctionCall(DistanceField=@n13) {materialFunction=&MF_UI_AntiAliasing}.Inner) {description=Output the mask, outputName=Shape Mask}
  FunctionOutput(None=@n1) {description=Output the UVs, outputName=UVs, sortPriority=50}
  n14 = FunctionInput() {
   bUsePreviewValueAsDefault=true
   description=Define the thickness of the outline
   inputName=Outline Thickness
   inputType=FunctionInput_Scalar
   sortPriority=20
   previewValue=<0.2 0.2 0.2 0.2>
  }
  n15 = Divide(A=@n14, B=NamedRerouteUsage())
  n16 = NamedRerouteUsage()
  n17 = FunctionOutput(None=MaterialFunctionCall!(DistanceField=@n13, Thickness=Multiply(A=@n15, B=@n16), Alignment=Constant() {r=-1}, Smoothing) {materialFunction=/FortUILibrary/MaterialFunctions/Shapes/Modifiers/MF_UI_ShapeOutline}.SDFOutline) {description=Output the outline SDF, outputName=Outline SDF, sortPriority=11}
  FunctionOutput(None=MaterialFunctionCall(DistanceField=@n17) {materialFunction=&MF_UI_AntiAliasing}.Inner) {description=Output the outline mask, outputName=Outline Mask, sortPriority=1}
  MaterialFunctionCall() {materialFunction=/FortUILibrary/MaterialFunctions/Basics/MF_UI_ViewportScale}
  n18 = NamedRerouteUsage()
  n19 = ComponentMask(None=@n18) {r=true}
  n20 = ComponentMask(None=@n18) {g=true}
  NamedRerouteDeclaration(None=Max(A=Min(A=@n19, B=@n20)) {constB=0}) {name=Min Pixel Size, nodeColor=<rgba 0 0 0 1>}
  n21 = FunctionInput() {
   bUsePreviewValueAsDefault=true
   description=Defines the skew angle
   inputName=Skew Angle
   inputType=FunctionInput_Scalar
   sortPriority=5
   previewValue=<0 0 0 0>
  }
  n22 = Multiply(
   A=ComponentMask(None=MaterialFunctionCall!(Dimensions) {materialFunction=/FortUILibrary/MaterialFunctions/Basics/MF_UI_GetAspectRatio}.V Ratio) {g=true}
   B=NamedRerouteDeclaration(None=@n21) {name=Skew Angle, nodeColor=<rgba 0 0 0 1>}
  )
  n23 = Constant()
  NamedRerouteDeclaration(None=MaterialFunctionCall!(UV=NamedRerouteUsage(), Skew Angle=AppendVector(A=@n22, B=@n23)) {materialFunction=/FortUILibrary/MaterialFunctions/Basics/MF_UI_Skew}.Skewed UVs) {name=Skewed UVs, nodeColor=<rgba 0 0 0 1>}
  n24 = NamedRerouteUsage()
  n25 = NamedRerouteUsage()
  n26 = Multiply(A=@n25, B=Constant2Vector() {g=0.5, r=0.5})
  n27 = Multiply(A=@n25, B=Constant2Vector() {g=0.5, r=-0.5})
  n28 = MaterialFunctionCall!(UV=NamedRerouteUsage()) {materialFunction=/FortUILibrary/MaterialFunctions/Basics/MF_UI_CornerMask}
  n29 = LinearInterpolate(
   A=NamedRerouteDeclaration(None=Subtract(A=@n24, B=@n26)) {name=Bottom Right UV, nodeColor=<rgba 0 0 0 1>}
   B=NamedRerouteDeclaration(None=Subtract(A=@n24, B=@n27)) {name=Bottom Left UV, nodeColor=<rgba 0 0 0 1>}
   Alpha=@n28.Bottom Left Mask
  ) {constA=0.5}
  n30 = Constant2Vector() {g=-0.5, r=0.5}
  n31 = NamedRerouteDeclaration(None=Subtract(A=@n24, B=Multiply(A=@n30, B=@n25))) {name=Top Right UV, nodeColor=<rgba 0 0 0 1>}
  n32 = Subtract(A=@n24, B=Multiply(A=Constant2Vector() {g=-0.5, r=-0.5}, B=@n25))
  n33 = NamedRerouteDeclaration(None=LinearInterpolate(
    A=LinearInterpolate(A=@n29, B=@n31, Alpha=@n28.Top Right Mask) {constA=0.5}
    B=NamedRerouteDeclaration(None=@n32) {name=Top Left UV, nodeColor=<rgba 0 0 0 1>}
    Alpha=@n28.Top Left Mask
   ) {constA=0.25}) {name=Combined Corners UV, nodeColor=<rgba 0 0 0 1>}
  n34 = NamedRerouteUsage()
  n35 = LinearInterpolate(
   A=LinearInterpolate(A=ComponentMask(None=@n34) {a=true}, B=ComponentMask(None=@n34) {g=true}, Alpha=@n28.Bottom Left Mask) {constA=0.5}
   B=ComponentMask(None=@n34) {b=true}
   Alpha=@n28.Top Right Mask
  ) {constA=0.5}
  n36 = ComponentMask(None=@n34) {r=true}
  n37 = Saturate(None=MaterialFunctionCall!(
    UV=MaterialFunctionCall(UV=@n33) {materialFunction=&MF_UI_UVScaling}.Scaled to Fit
    Size=LinearInterpolate(A=@n35, B=@n36, Alpha=@n28.Top Left Mask) {constA=0.25}
    OutlineThickness
    OutlineAlignment
    GlowMin
    GlowMax
   ) {materialFunction=&MF_UI_Circle}.SDF)
  NamedRerouteDeclaration(None=OneMinus(None=Ceil(None=@n37))) {name=Corner Aliasing Mask, nodeColor=<rgba 0 0 0 1>}
  n38 = MaterialFunctionCall() {materialFunction=&GetUserInterfaceUV}
  NamedRerouteDeclaration(None=FunctionInput(Preview=@n38.Normalized UV) {bUsePreviewValueAsDefault=true, description=Define the UVs, inputName=UV, inputType=FunctionInput_Vector2}) {name=UV, nodeColor=<rgba 0 0 0 1>}
  n39 = NamedRerouteDeclaration(None=Constant() {r=0.0075}) {name=Edge Falloff Value, nodeColor=<rgba 0 0 0 1>}
  n40 = Divide(B=NamedRerouteUsage()) {constA=192, constB=1}
  n41 = Max(A=Multiply(A=Multiply(A=@n39, B=@n40), B=NamedRerouteUsage())) {constB=0}
  NamedRerouteDeclaration(None=Multiply(A=NamedRerouteDeclaration(None=@n41) {name=Max Value, nodeColor=<rgba 0 0 0 1>}) {constB=-1}) {name=Min Value, nodeColor=<rgba 0 0 0 1>}
  n42 = ViewSize()
  n43 = ComponentMask(None=@n42) {r=true}
  n44 = ComponentMask(None=@n42) {g=true}
  n45 = MaterialFunctionCall!(A=@n43, B=@n44) {materialFunction=/FortUILibrary/MaterialFunctions/Basics/MF_UI_BooleanOperations}
  n46 = Constant2Vector() {g=1080, r=1920}
  NamedRerouteDeclaration(None=Divide(A=LinearInterpolate(A=@n43, B=@n44, Alpha=@n45.IsGreater), B=LinearInterpolate(A=@n46.R, B=@n46.R, Alpha=@n45.IsGreater))) {name=Resolution Factor, nodeColor=<rgba 0 0 0 1>}
  NamedRerouteDeclaration(None=@n38.Pixel Size) {name=Pixel Size, nodeColor=<rgba 0 0 0 1>}
  n47 = NamedRerouteUsage()
  NamedRerouteDeclaration(None=LinearInterpolate(B=NamedRerouteUsage(), Alpha=Floor(None=@n47)) {constA=1}) {name=Aliasing Mask, nodeColor=<rgba 0 0 0 1>}
  n48 = NamedRerouteUsage()
  n49 = NamedRerouteUsage()
  OneMinus(None=Saturate(None=SmoothStep(Min=@n48, Max=@n49, Value=@n13)))
  OneMinus(None=Saturate(None=SmoothStep(Min=@n48, Max=@n49, Value=@n17)))
 ]
}
