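# Translucent UI material that draws a rounded outline stroke, named as a background stroke for Jam UI, with tinting for button states.
# MF_UI_JamCoreUI_RoundedMask supplies a fixed-thickness outline mask from the corner radius parameters.
# MF_UI_JamButtonStatesBase_Lerp supplies state color and opacity.
# The state color multiplies the Color parameter, while the state opacity multiplies the outline mask and Opacity parameter.
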Material /FMJamSystem/Widgets/Materials/M_UI_JamBGStroke
graph {
 properties {bAutomaticallySetUsageInEditor=false, blendMode=BLEND_Translucent, materialDomain=MD_UI, usedShadingModels=Default Lit}
 nodes [
  n1 = MaterialFunctionCall!(Idle, Idle Opacity, Hovered, Hovered Opacity, Focused, Focused Opacity, Pressed, Pressed Opacity, Disabled, Disabled Opacity) {materialFunction=&/MF_UI_JamButtonStatesBase_Lerp}
  n2 = VectorParameter() {group=Corners, parameterName=Corner Radius, sortPriority=2, defaultValue=<rgba 0.25 0.25 0.25 0.25>}
  n3 = AppendVector(A=@n2.RGB, B=@n2.RGB)
  n4 = ScalarParameter() {group=Corners, parameterName=Radius in Pixels, sliderMax=1, sortPriority=0}
  n5 = Constant() {r=4}
  Multiply(
   A=Multiply(
    A=@n1.Opacity
    B=MaterialFunctionCall!(UV, Size, Corner Radius=@n3, Radius in Pixels=@n4, Skew Angle, Outline Thickness=@n5) {materialFunction=&/MF_UI_JamCoreUI_RoundedMask}.Outline Mask
   )
   B=ScalarParameter() {defaultValue=0.3, parameterName=Opacity}
  )
  Multiply(A=VectorParameter() {parameterName=Color, defaultValue=<rgba 0 0 0 1>}.RGB, B=@n1.Color)
 ]
}
