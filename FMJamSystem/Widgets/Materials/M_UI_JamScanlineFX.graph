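# Translucent UI material for FMJam widgets that draws an animated CRT scanline and glow effect.
# Time-scrolled sine waves form fine scanlines and two sets of thin CRT lines offset by seeded randomness, blended into a combined scanline mask.
# Box SDFs with sine easing shape edge and inner glows, with 9-slice UVs declared as a reroute.
# Named reroutes carry the masks, and a color stage lerps tile background to glow color with a boosted highlight.
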Material /FMJamSystem/Widgets/Materials/M_UI_JamScanlineFX
graph {
 properties {bAutomaticallySetUsageInEditor=false, blendMode=BLEND_Translucent, materialDomain=MD_UI, usedShadingModels=Default Lit}
 nodes [
  n1 = Multiply(A=MaterialFunctionCall!(UVs, Radius) {materialFunction=/FortUILibrary/MaterialFunctions/Shapes/SDFs/MF_SDF_Circle}.SDF) {constB=2}
  Multiply(A=OneMinus(None=NamedRerouteUsage()), B=OneMinus(None=Saturate(None=@n1)))
  n2 = Floor(None=ScalarParameter() {defaultValue=62, group=Scanlines, parameterName=Scanline Count})
  n3 = MaterialFunctionCall() {materialFunction=&GetUserInterfaceUV}
  n4 = Frac(None=Multiply(A=Time(), B=ScalarParameter() {defaultValue=0.15, group=Scanlines, parameterName=Scanline Speed, sliderMax=0.25}))
  n5 = Multiply(A=Add(A=Add(A=@n3.Normalized UV, B=@n4)) {constB=2}) {constB=0.2}
  n6 = ComponentMask(None=Sine(None=Multiply(A=@n2, B=@n5))) {g=true}
  n7 = ScalarParameter() {defaultValue=0.06, group=Scanlines, parameterName=Scanline Thickness, sliderMax=0.1}
  MaterialExpressionRemap_0 = MaterialXRemap!(None=Saturate(None=Subtract(A=@n6, B=@n7) {constB=0.5}), InputLow, InputHigh=@n7, TargetLow, TargetHigh)
  NamedRerouteDeclaration(None=@MaterialExpressionRemap_0) {name=Scanlines, nodeColor=<rgba 0.5395 0 0.9647 1>}
  n8 = MaterialFunctionCall(UVs=NamedRerouteUsage(), Size=ScalarParameter() {defaultValue=0.7, group=Edge Glow, parameterName=Edge Glow Size, sliderMax=1}) {materialFunction=&MF_UI_Box}
  NamedRerouteDeclaration(None=MaterialFunctionCall("Value (0-1)"=Saturate(None=@n8.SDF)) {materialFunction=&MF_UI_EaseCurves}.EaseInOutSine) {name=Edge Glow, nodeColor=<rgba 0.2338 0.9647 0.7401 1>}
  n9 = Multiply(A=Time(), B=ScalarParameter() {defaultValue=0.6, group=CRT Lines, parameterName=CRT Thin 1 Speed, sliderMax=1})
  n10 = MaterialFunctionCall() {materialFunction=&GetUserInterfaceUV}
  MaterialExpressionRemap_1 = MaterialXRemap(
   None=MaterialFunctionCall!(UV=Floor(None=@n10.Normalized UV), Seed=ScalarParameter() {defaultValue=8, parameterName=RandomSeed}) {materialFunction=&MF_UI_Random}.Result
   InputLow=Constant()
   InputHigh=Constant() {r=1}
   TargetLow=Constant()
   TargetHigh=Constant() {r=1}
  )
  n11 = Sine(None=Add(A=Panner(Coordinate=@n10.Normalized UV, Time=@n9) {speedY=0.575}, B=@MaterialExpressionRemap_1))
  n12 = ScalarParameter() {defaultValue=128, group=CRT Lines, parameterName=CRT Hardness}
  n13 = Multiply(A=Subtract(A=ComponentMask(None=@n11) {g=true}) {constB=0.993}, B=@n12)
  n14 = Multiply(A=Time(), B=ScalarParameter() {defaultValue=0.8, group=CRT Lines, parameterName=CRT Thin 2 Speed, sliderMax=1})
  n15 = Sine(None=Add(A=Panner(Coordinate=@n10.Normalized UV, Time=@n14) {speedY=0.575}, B=@MaterialExpressionRemap_1))
  n16 = Multiply(A=Subtract(A=ComponentMask(None=@n15) {g=true}) {constB=0.97}, B=@n12)
  NamedRerouteDeclaration(None=Add(A=Saturate(None=@n13), B=Saturate(None=@n16))) {name=CRT Lines, nodeColor=<rgba 0.9647 0.1912 0.3095 1>}
  n17 = Multiply(A=NamedRerouteUsage(), B=ScalarParameter() {defaultValue=0.5, group=CRT Lines, parameterName=CRT Lines Opacity})
  n18 = OneMinus(None=Multiply(A=NamedRerouteUsage(), B=ScalarParameter() {defaultValue=0.025, group=Scanlines, parameterName=Scanline Opacity}) {constB=0.1})
  NamedRerouteDeclaration(None=Multiply(A=Add(A=@n17, B=@n18))) {name=Scanline Combo, nodeColor=<rgba 1 0 0.375 1>}
  NamedRerouteDeclaration(None=@n3.9-Slice UV) {name=9 Slice UV, nodeColor=<rgba 0 0 0 1>}
  n19 = ScalarParameter() {defaultValue=0.95, group=Inner Glow, parameterName=Inner Glow Size, sliderMax=1}
  n20 = Saturate(None=MaterialFunctionCall(UVs=NamedRerouteUsage(), Size=OneMinus(None=@n19)) {materialFunction=&MF_UI_Box}.SDF)
  NamedRerouteDeclaration(None=Multiply(
    A=MaterialFunctionCall("Value (0-1)"=@n20) {materialFunction=&MF_UI_EaseCurves}.EaseInOutSine
    B=ScalarParameter() {defaultValue=0.15, group=Inner Glow, parameterName=Inner Glow Opacity, sliderMax=1}
   )) {name=Inner Glow, nodeColor=<rgba 0 0.4994 0.9647 1>}
  Power(Base=VectorParameter() {parameterName=Color, defaultValue=<rgba 0.1295 0.8325 0.9479 1>}.RGB)
  n21 = VectorParameter() {parameterName=Tile Bkgd Color, defaultValue=<rgba 0 0 0 1>}
  n22 = NamedRerouteUsage()
  n23 = VectorParameter() {parameterName=Tile Glow Color, defaultValue=<rgba 1 0 0.8634 1>}
  n24 = Saturate(None=Subtract(A=@n22) {constB=0.75})
  Saturate(None=Add(A=LinearInterpolate(A=@n21.RGB, B=@n23.RGB, Alpha=@n22), B=Multiply(A=@n24) {constB=0.5}))
 ]
}
