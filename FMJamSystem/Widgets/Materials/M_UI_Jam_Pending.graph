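# M_UI_Jam_Pending is a translucent, unlit UI material that draws an animated circular pending indicator for the Jam system's widgets.
# Time and Speed rotate fitted UVs, which feed SDF circle, outline and radial-arc functions with anti-aliasing.
# These build a background ring and a partial inner arc with an end point, fixed at a constant progress value.
# Parameters set thicknesses, background color and opacity, and two blended arc colors.
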Material /FMJamSystem/Widgets/Materials/M_UI_Jam_Pending
graph {
 properties {blendMode=BLEND_Translucent, materialDomain=MD_UI, pixelDepthOffsetMode=PDOM_Legacy, shadingModel=MSM_Unlit, usedShadingModels=Unlit}
 nodes [
  n1 = Multiply(
   A=ScalarParameter() {defaultValue=0.6, parameterName=Background Opacity, sliderMax=1}
   B=ScalarParameter() {defaultValue=1, parameterName=Display Background, sliderMax=1}
  )
  LinearInterpolate(
   A=Multiply(A=NamedRerouteUsage(), B=NamedRerouteDeclaration(None=@n1) {name=Background Opacity, nodeColor=<rgba 0 0 0 1>})
   Alpha=NamedRerouteUsage()
  )
  n2 = NamedRerouteUsage()
  LinearInterpolate(
   A=Multiply(A=VectorParameter() {parameterName=Background Color, defaultValue=<rgba 0.9104 0 1 1>}.RGB, B=Ceil(None=@n2))
   B=LinearInterpolate(
    A=VectorParameter() {parameterName=Color A, defaultValue=<rgba 1 0.0265 0 1>}.RGB
    B=VectorParameter() {parameterName=Color B, defaultValue=<rgba 0.2793 1 0 1>}.RGB
    Alpha=NamedRerouteUsage()
   )
   Alpha=NamedRerouteUsage()
  )
  n3 = MaterialFunctionCall(UV=MaterialFunctionCall() {materialFunction=&GetUserInterfaceUV}.Normalized UV) {materialFunction=&MF_UI_UVScaling}
  n4 = Frac(None=Multiply(A=Time(), B=ScalarParameter() {defaultValue=0.75, parameterName=Speed}) {constB=0.75})
  NamedRerouteDeclaration(None=Rotator(Coordinate=NamedRerouteDeclaration(None=@n3.Scaled to Fit) {name=UV, nodeColor=<rgba 0 0 0 1>}, Time=OneMinus(None=@n4)) {speed=6.2832}) {name=Rotating UV, nodeColor=<rgba 0 0 0 1>}
  n5 = NamedRerouteUsage()
  n6 = NamedRerouteDeclaration(None=Constant() {r=0.99}) {name=Background Size, nodeColor=<rgba 0 0 0 1>}
  n7 = NamedRerouteDeclaration(None=ScalarParameter() {defaultValue=0.175, parameterName=Background Thickness, sliderMax=1}) {name=Outline Thickness, nodeColor=<rgba 0 0 0 1>}
  n8 = Add(
   A=NamedRerouteDeclaration(None=MaterialFunctionCall(UVs=@n5, Size=@n6) {materialFunction=&MF_UI_Circle}.SDF) {name=Shape SDF, nodeColor=<rgba 0 0 0 1>}
   B=@n7
  )
  NamedRerouteDeclaration(None=Subtract(A=Abs(None=@n8), B=@n7) {constB=0.1}) {name=Outer Outline SDF, nodeColor=<rgba 0 0 0 1>}
  NamedRerouteDeclaration(None=MaterialFunctionCall(DistanceField=NamedRerouteUsage()) {materialFunction=&MF_UI_AntiAliasing}.Inner) {name=Background Ring Mask, nodeColor=<rgba 0 0 0 1>}
  n9 = MaterialFunctionCall(DistanceField=NamedRerouteUsage()) {materialFunction=&MF_UI_AntiAliasing}
  n10 = MaterialFunctionCall(DistanceField=NamedRerouteUsage()) {materialFunction=&MF_UI_AntiAliasing}
  n11 = NamedRerouteUsage()
  n12 = MaterialFunctionCall!(
   UV=NamedRerouteUsage()
   "Arc Start Angle (0-1)"=Constant() {r=0.8}
   "Arc End Angle (0-1)"=Constant() {r=0.9}
   "Fill Ratio (0-1)"
   Fill Clockwise
  ) {materialFunction=/FortUILibrary/MaterialFunctions/Basics/MF_UI_RadialArc}
  NamedRerouteDeclaration(None=Multiply(
    A=LinearInterpolate(A=@n9.Inner, B=@n10.Inner, Alpha=@n11)
    B=NamedRerouteDeclaration(None=@n12.Radial Fill) {name=Inner Circle Gradient, nodeColor=<rgba 0 0 0 1>}
   )) {name=Inner Circle Mask, nodeColor=<rgba 0 0 0 1>}
  n13 = Rotator(Coordinate=NamedRerouteUsage(), Time=OneMinus(None=NamedRerouteUsage())) {speed=6.2832}
  n14 = Constant2Vector() {g=-0.5}
  n15 = NamedRerouteUsage()
  n16 = NamedRerouteUsage()
  n17 = AppendVector(A=@n14.R, B=Add(A=Multiply(A=@n14.G, B=@n15), B=Multiply(A=@n16) {constB=0.5})) {desc=Offset}
  n18 = NamedRerouteUsage()
  n19 = NamedRerouteUsage()
  n20 = Multiply(
   A=Saturate(None=Subtract(A=@n18, B=@n19))
   B=ScalarParameter() {defaultValue=0.125, desc=Inner Outline and End Point Feathering, parameterName=Inner Circle Thickness, sliderMax=1}
  )
  NamedRerouteDeclaration(None=MaterialFunctionCall(UVs=Subtract(A=@n13, B=@n17), Size=@n20) {materialFunction=&MF_UI_Circle}.SDF) {name=End Point, nodeColor=<rgba 0 0 0 1>}
  n21 = NamedRerouteUsage()
  n22 = OneMinus(None=NamedRerouteDeclaration(None=Constant() {r=0.75}) {name=Progress Value, nodeColor=<rgba 0 0 0 1>})
  NamedRerouteDeclaration(None=Ceil(None=MaterialFunctionCall(UV=@n21, "Fill Ratio (0-1)"=@n22) {materialFunction=&MF_UI_RadialArc}.Radial Fill)) {name=Progress Mask, nodeColor=<rgba 0 0 0 1>}
  n23 = NamedRerouteUsage()
  n24 = Saturate(None=Add(A=NamedRerouteUsage(), B=OneMinus(None=@n23)))
  NamedRerouteDeclaration(None=MaterialFunctionCall(UVs=NamedRerouteUsage(), Size=OneMinus(None=@n24), OutlineThickness=@n20) {materialFunction=&MF_UI_Circle}.SDFOutline) {name=Inner Outline SDF, nodeColor=<rgba 0 0 0 1>}
  n25 = Constant() {r=0.5}
  n26 = Constant() {r=0.5}
  n27 = Constant() {r=1.358}
  n28 = Constant() {r=-0.442}
  n29 = Constant() {r=-0.502}
  n30 = Constant() {r=-0.282}
  n31 = Constant() {r=1}
  n32 = Constant() {r=1}
  n33 = Constant() {r=1}
  n34 = Constant()
  n35 = Constant() {r=0.33}
  n36 = Constant() {r=0.67}
  NamedRerouteDeclaration(None=MaterialFunctionCall!(
    Gradient=MaterialFunctionCall!(UVs, Rotation) {materialFunction=/FortUILibrary/MaterialFunctions/Gradients/MF_UI_AngleGradient}.Result
    a=MaterialFunctionCall(R=@n25, G=@n26, B=@n27) {desc=a, materialFunction=&AppendMany}.RGB
    b=MaterialFunctionCall(R=@n28, G=@n29, B=@n30) {desc=b, materialFunction=&AppendMany}.RGB
    c=MaterialFunctionCall(R=@n31, G=@n32, B=@n33) {desc=c, materialFunction=&AppendMany}.RGB
    d=MaterialFunctionCall(R=@n34, G=@n35, B=@n36) {desc=d, materialFunction=&AppendMany}.RGB
   ) {materialFunction=/FortUILibrary/MaterialFunctions/Color/MF_UI_CosineGradient}.Result) {name=Rift Energy, nodeColor=<rgba 0 0 0 1>}
 ]
}
