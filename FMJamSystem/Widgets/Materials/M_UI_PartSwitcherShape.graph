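# Translucent UI material that draws the part switcher shape: a horizontal strip with a movable circle on it.
# A box function sized by Strip Size masks a reflected gradient to form the strip, while a circle function draws the circle from UVs shifted horizontally by Circle Position.
# The circle's fill, scaled by Circle Opacity, cuts out the dimmed strip beneath it and is added back at its own fixed strength.
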Material /FMJamSystem/Widgets/Materials/M_UI_PartSwitcherShape
graph {
 properties {bAutomaticallySetUsageInEditor=false, blendMode=BLEND_Translucent, materialDomain=MD_UI, usedShadingModels=Default Lit}
 nodes [
  MaterialExpressionTextureCoordinate_0 = TextureCoordinate()
  n1 = Convert(None=@MaterialExpressionTextureCoordinate_0) {
   nodeName=Break Float2
   convertInputs [{
     type=Vector2
     expressionInput {
      expression=&self:MaterialExpressionTextureCoordinate_0
      inputName=None
      mask=0
      maskA=0
      maskB=0
      maskG=0
      maskR=0
      outputIndex=0
     }
    }]
   convertOutputs [{type=Scalar}, {type=Scalar}]
  }
  n2 = Add(A=ScalarParameter() {parameterName=Circle Position}, B=@n1) {constB=0}
  n3 = Multiply(
   A=MaterialFunctionCall(UVs=AppendVector(A=@n2, B=@n1), Size=ScalarParameter() {defaultValue=0.355, parameterName=Circle Size}) {materialFunction=&MF_UI_Circle}.Fill
   B=ScalarParameter() {defaultValue=1, parameterName=Circle Opacity}
  )
  n4 = LinearInterpolate(
   B=MaterialFunctionCall!(UVs, Offset, RotationCenter, RotationAngle) {materialFunction=/FortUILibrary/MaterialFunctions/Gradients/MF_UI_ReflectedGradient}.Result
   Alpha=MaterialFunctionCall(Size=VectorParameter() {parameterName=Strip Size, defaultValue=<rgba 1 0.18 0 1>}.RGB) {materialFunction=&MF_UI_Box}.Fill
  )
  Add(A=Multiply(A=@n3) {constB=0.6}, B=LinearInterpolate(A=Multiply(A=@n4) {constB=0.5}, Alpha=@n3) {constB=0})
  Constant3Vector()
 ]
}
