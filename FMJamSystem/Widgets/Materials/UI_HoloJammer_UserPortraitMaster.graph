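# Translucent UI master material for HoloJammer user portraits, with an optional hologram look that remaps desaturated brightness through a three-color blue-to-cyan ramp.
# A large or small preview texture is composited over a gradient background, with inactive and empty states, tinting and optional stroke.
# Opacity combines a switchable round or clipped portrait mask with background alpha.
# An optional reveal mask smoothsteps a triangle pattern against radial gradients and three positionable shatter-hole decals.
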Material /FMJamSystem/Widgets/Materials/UI_HoloJammer_UserPortraitMaster
graph {
 properties {blendMode=BLEND_Translucent, materialDomain=MD_UI, numCustomizedUVs=3, pixelDepthOffsetMode=PDOM_Legacy, usedShadingModels=Default Lit}
 nodes [
  MaterialFunctionCall(Texture Scale=ScalarParameter() {defaultValue=1, parameterName=GlobalScale}) {materialFunction=&ScaleUVsByCenter}
  n1 = MaterialFunctionCall!(Native_U_PixelSize=Constant() {r=250}) {materialFunction=/FortUILibrary/Foundation/Art_ToolKit/Tools/M_UI_Resolution_Sharpness_Correction}
  n2 = MaterialFunctionCall(
   Radius=ScalarParameter() {defaultValue=0.5, parameterName=MaskRevealRadius}
   Density=StaticSwitchParameter(True=Multiply(A=@n1.Result) {constB=0.05}, False=ScalarParameter() {defaultValue=10, parameterName=MaskRevealDensity}) {parameterName=AutoDensity}
  ) {materialFunction=&RadialGradientExponential}
  n3 = TextureCoordinate()
  n4 = ScalarParameter() {group=PortaitMask, parameterName=Mask Rotation}
  n5 = MaterialFunctionCall(
   UVs=StaticSwitchParameter(
    True=@n2.RadialGradientExponential
    False=MaterialFunctionCall(UVs=@n3, "Rotation Angle (0-1)"=@n4) {materialFunction=&CustomRotator}.Rotated Values
   ) {parameterName=UseRoundMask}
   Texture Scale=ScalarParameter() {defaultValue=1, group=PortaitMask, parameterName=Mask Size}
  ) {materialFunction=&ScaleUVsByCenter}
  n6 = AppendVector(A=ScalarParameter() {defaultValue=1, group=PortaitMask, parameterName=Mask Side Clip}, B=Constant() {r=1})
  n7 = NamedRerouteUsage()
  n8 = TextureCoordinate()
  n9 = OneMinus(None=Round(None=ComponentMask(None=@n8) {g=true}))
  n10 = Max(
   A=Multiply(
    A=MaterialFunctionCall(UVs=@n3, Texture Scale=@n6) {materialFunction=&ScaleUVsByCenter}.0-1 mask
    B=Multiply(A=@n7, B=@n9)
   )
   B=@n5.0-1 mask
  )
  n11 = TextureSample(UVs=TextureCoordinate() {coordinateIndex=2}) {texture=/Game/Global/Textures/Default/Masks_and_Gradients/T_TrianglePattern}
  n12 = VectorParameter() {group=MaskReveal, parameterName=RevealDecal1}
  n13 = MaterialFunctionCall!(Texture Scale=@n12.A, UVs=TextureCoordinate(), Offset XY Texturesize Z=@n12.A) {materialFunction=/Engine/Functions/Engine_MaterialFunctions01/Texturing/ScaleUVsAroundPoint}
  n14 = VectorParameter() {group=MaskReveal, parameterName=RevealDecal2}
  n15 = MaterialFunctionCall(Texture Scale=@n14.A, UVs=TextureCoordinate(), Offset XY Texturesize Z=@n14.A) {materialFunction=&ScaleUVsAroundPoint}
  n16 = VectorParameter() {group=MaskReveal, parameterName=RevealDecal3}
  n17 = TextureCoordinate()
  n18 = Max(
   A=Max(
    A=TextureSample(UVs=@n13.UVs) {samplerSource=SSM_Clamp_WorldGroupSettings, texture=/Game/Athena/HUD/Discoverability/materials/T_UI_Shatterhole}.R
    B=TextureSample(UVs=@n15.UVs) {samplerSource=SSM_Clamp_WorldGroupSettings, texture=&T_UI_Shatterhole}.R
   )
   B=TextureSample(UVs=MaterialFunctionCall(Texture Scale=@n16.A, UVs=@n17, Offset XY Texturesize Z=@n16.A) {materialFunction=&ScaleUVsAroundPoint}.UVs) {samplerSource=SSM_Clamp_WorldGroupSettings, texture=&T_UI_Shatterhole}.R
  )
  n19 = OneMinus(None=MaterialFunctionCall(
    Radius=ScalarParameter() {defaultValue=0.2, group=MaskReveal, parameterName=InnerMaskRevealRadius}
    Density=ScalarParameter() {defaultValue=10, group=MaskReveal, parameterName=InnerMaskRevealDensity}
   ) {materialFunction=&RadialGradientExponential}.RadialGradientExponential)
  n20 = Max(A=@n18, B=OneMinus(None=Multiply(A=@n19, B=@n2.RadialGradientExponential)))
  n21 = Add(A=@n20, B=Multiply(A=@n20, B=ScalarParameter() {group=MaskReveal, parameterName=MaskEdgeSoftness}))
  n22 = Multiply(B=Multiply(
    A=@n10
    B=MaterialFunctionCall(Alpha=@n11.R, Min=@n20, Max=@n21) {materialFunction=&SmoothStep}.Result
   )) {constA=1}
  n23 = LinearInterpolate(A=ScalarParameter() {defaultValue=1, parameterName=BackgroundAlpha}, Alpha=NamedRerouteUsage())
  n24 = Constant() {r=1}
  Multiply(
   A=StaticSwitchParameter(True=@n5.0-1 mask, False=StaticSwitchParameter(True=@n22, False=@n10) {parameterName=UseMask}) {parameterName=UseRoundMask}
   B=StaticSwitchParameter(True=StaticSwitchParameter(True=@n23, False=@n24) {parameterName=UseAlphaBackground}, False=NamedRerouteUsage()) {defaultValue=true, parameterName=UseColorBackground}
  )
  n25 = Constant3Vector() {constant=<rgba 0 0.0258 0.2969 1>}
  n26 = NamedRerouteUsage()
  n27 = Multiply(A=Desaturation(None=Power(Base=@n26) {constExponent=1.5})) {constB=1.5}
  n28 = Constant3Vector() {constant=<rgba 0.1295 0.8325 0.9479 1>}
  n29 = Constant3Vector() {constant=<rgba 0.6406 1 0.9785 1>}
  NamedRerouteDeclaration(None=Saturate(None=MaterialFunctionCall(A=@n25, B=@n28, C=@n29, Alpha=@n27) {materialFunction=&Lerp_3Color}.Result)) {name=Holo FX Color, nodeColor=<rgba 1 0 0.4891 1>}
  n30 = TextureCoordinate()
  n31 = ScalarParameter() {defaultValue=1.75, parameterName=PortraitScale}
  n32 = Constant()
  n33 = ScalarParameter() {defaultValue=-0.175, parameterName=PortraitOffset}
  NamedRerouteDeclaration(None=Add(
    A=MaterialFunctionCall(UVs=@n30, Texture Scale=@n31) {materialFunction=&ScaleUVsByCenter}.UVs
    B=AppendVector(A=@n32, B=@n33)
   ) {constB=0}) {name=Portrait Scaled, nodeColor=<rgba 0.0781 0 1 1>}
  n34 = LinearInterpolate(A=TextureCoordinate() {coordinateIndex=1}, B=@n30, Alpha=ScalarParameter() {parameterName=CustomUV})
  n35 = TextureSampleParameter2D(UVs=@n34) {
   parameterName=LargePreview
   samplerSource=SSM_Clamp_WorldGroupSettings
   texture=/Game/UI/Foundation/Textures/Icons/Heroes/Athena/Soldier/T-Soldier-HID-005-Athena-Commando-M-L
  }
  n36 = TextureSampleParameter2D(UVs=@n34) {
   parameterName=SmallPreview
   samplerSource=SSM_Clamp_WorldGroupSettings
   texture=/Game/UI/Foundation/Textures/Icons/Heroes/Athena/Soldier/T-Soldier-HID-005-Athena-Commando-M
  }
  NamedRerouteDeclaration(None=StaticSwitchParameter(
    True=MaterialFunctionCall!(UVs=@n34, Alpha=@n35.A) {materialFunction=/Game/UI/Foundation/Items/ClampAlphaToTransparentBorder}.Result
    False=MaterialFunctionCall(UVs=@n34, Alpha=@n36.A) {materialFunction=&ClampAlphaToTransparentBorder}.Result
   ) {parameterName=UseLargePreviewImage}) {name=Use Large Preview Image, nodeColor=<rgba 1 0 0.2812 1>}
  n37 = VectorParameter() {group=Stroke, parameterName=StrokeColor}
  n38 = MaterialFunctionCall(
   Radius=ScalarParameter() {defaultValue=0.5, group=Stroke, parameterName=StrokeRadius}
   Density=ScalarParameter() {defaultValue=10, group=Stroke, parameterName=StrokeDensity}
  ) {materialFunction=&RadialGradientExponential}
  n39 = VectorParameter() {parameterName=BackgroundRGB, defaultValue=<rgba 0.013 0.3613 0.6376 1>}
  n40 = Multiply(A=MaterialFunctionCall(CenterPosition=Constant2Vector() {g=0.75, r=0.5}, Radius=Constant() {r=0.75}, Density=Constant() {r=1.25}) {materialFunction=&RadialGradientExponential}.RadialGradientExponential) {constB=0.5}
  n41 = LinearInterpolate(A=@n39.RGB, B=VectorParameter() {parameterName=HighlightRGB, defaultValue=<rgba 0.4452 1 1 1>}.RGB, Alpha=@n40)
  n42 = LinearInterpolate(
   A=LinearInterpolate(A=VectorParameter() {parameterName=InactiveBackgroundRGB, defaultValue=<rgba 0 0.0999 0.3095 1>}.RGB, B=@n39.RGB, Alpha=@n40)
   B=LinearInterpolate(A=@n41, B=VectorParameter() {parameterName=InactiveRGBALerp, defaultValue=<rgba 0 0.0168 0.0887 0.5>}.RGB, Alpha=Constant() {r=0.75})
   Alpha=ScalarParameter() {parameterName=IsKairosPortrait}
  )
  n43 = ScalarParameter() {parameterName=IsInactive, sliderMax=1}
  n44 = LinearInterpolate(
   A=VectorParameter() {parameterName=NoPortraitFill, defaultValue=<rgba 0 0.0999 0.3095 1>}.RGB
   B=VectorParameter() {parameterName=NoPortraitHighlight, defaultValue=<rgba 0 0.3228 1 1>}.RGB
   Alpha=@n40
  )
  n45 = StaticSwitchParameter(
   True=LinearInterpolate(A=@n35.RGB, B=@n44, Alpha=ScalarParameter() {parameterName=IsEmpty})
   False=LinearInterpolate(A=@n36.RGB, B=@n44, Alpha=ScalarParameter() {parameterName=IsEmpty})
  ) {parameterName=UseLargePreviewImage}
  n46 = LinearInterpolate(A=ScalarParameter() {defaultValue=0.5, parameterName=DesaturationFactor}, B=Constant(), Alpha=ScalarParameter() {parameterName=IsEmpty})
  n47 = VectorParameter() {parameterName=InactiveRGBALerp, defaultValue=<rgba 0 0.0168 0.0887 0.5>}
  n48 = VectorParameter() {parameterName=PortraitFill}
  n49 = LinearInterpolate(
   A=LinearInterpolate(A=@n45, B=@n48.RGB, Alpha=@n48.RGB)
   B=LinearInterpolate(A=Desaturation(None=@n45, Fraction=@n46), B=@n47.RGB, Alpha=@n47.RGB)
   Alpha=@n43
  )
  n50 = StaticSwitchParameter(True=LinearInterpolate(A=LinearInterpolate(A=@n41, B=@n42, Alpha=@n43), B=@n49, Alpha=NamedRerouteUsage()), False=@n49) {defaultValue=true, parameterName=UseColorBackground}
  NamedRerouteDeclaration(None=LinearInterpolate(
    A=@n50
    B=LinearInterpolate(A=@n37.RGB, B=@n50, Alpha=@n38.RadialGradientExponential) {constAlpha=0}
    Alpha=ScalarParameter() {group=Stroke, parameterName=UseStroke}
   ) {constAlpha=0}) {name=Final Color, nodeColor=<rgba 1 0.4219 0 1>}
  NamedRerouteUsage()
  OneMinus(None=@n18)
  StaticSwitchParameter(True=NamedRerouteUsage(), False=NamedRerouteUsage()) {defaultValue=true, parameterName=HoloFX}
  TextureCoordinate()
 ]
}
