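# WBP_FMJam_Reticle is a HUD element widget for the FMJam system that shows a reticle at the center of its area.
# A root Overlay centers a single Image, Image_Dot, drawn with the M_UI_FM_Reticle material from FMCore.
# The event graph and variable list are empty, so the widget adds no Blueprint logic to its FortHUDElementWidget parent.
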WidgetBlueprint /FMJamSystem/Widgets/WBP_FMJam_Reticle
graphs [EventGraph {links [], nodes []}]
variables []
widgets {
 info {inheritedWidgetCount=1, parentClass=&FortHUDElementWidget, rootWidgetClass=&Overlay, widgetCount=3}
 widgets [
  Overlay_Root {widgetClassPath=&Overlay, widget=Overlay_Root}
  Image_Dot {
   widgetClassPath=&Image
   bIsVariable=true
   parent=Overlay_Root
   widget=Image_Dot
   properties {brush {
     bIsDynamicallyLoaded=false
     drawAs=Image
     imageType=FullColor
     mirroring=NoMirror
     resourceName=None
     resourceObject=/FMCore/Materials/M_UI_FM_Reticle
     tiling=NoTile
     imageSize=<256 256>
     margin=<ltrb 0 0 0 0>
     outlineSettings {
      bUseBrushTransparency=false
      roundingType=HalfHeightRadius
      width=0
      color {colorUseRule=UseColor_Specified, specifiedColor=<rgba 0 0 0 0>}
      cornerRadii=<0 0 0 0>
     }
     tintColor {colorUseRule=UseColor_Specified, specifiedColor=<rgba 1 1 1 1>}
     uVRegion {bIsValid=false, max=<0 0>, min=<0 0>}
    }}
   slot {class=OverlaySlot, properties {horizontalAlignment=HAlign_Center, verticalAlignment=VAlign_Center}}
  }
  TutorialHighlight {widgetClassPath=&TutorialHighlight}
 ]
}
