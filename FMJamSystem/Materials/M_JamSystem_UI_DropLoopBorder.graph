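# Translucent UI material for the Jam System, named as the drop loop border, that builds a gradient-tinted, rounded outline.
# MF_UI_LinearGradient drives a lerp from a purple-blue GradientColor toward the default B value, with parameterized start, end and rotation.
# MF_UI_CoreUI_RoundedMask shapes the border from an outline thickness scalar and only the red channel of the corner radius vector.
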Material /FMJamSystem/Materials/M_JamSystem_UI_DropLoopBorder
graph {
 properties {blendMode=BLEND_Translucent, materialDomain=MD_UI, pixelDepthOffsetMode=PDOM_Legacy, usedShadingModels=Default Lit}
 nodes [
  LinearInterpolate(
   A=VectorParameter() {group=Gradient Settings, parameterName=GradientColor, defaultValue=<rgba 0.2918 0.227 0.9301 1>}.RGB
   Alpha=MaterialFunctionCall(
    GradientStart=ScalarParameter() {group=Gradient Settings, parameterName=GradientStart, sortPriority=1}
    GradientEnd=ScalarParameter() {defaultValue=1, group=Gradient Settings, parameterName=GradientEnd, sortPriority=2}
    RotationAngle=ScalarParameter() {defaultValue=0.2, group=Gradient Settings, parameterName=GradientRotation, sortPriority=4}
   ) {materialFunction=&MF_UI_LinearGradient}.Result
  )
  MaterialFunctionCall(
   Corner Radius=VectorParameter() {parameterName="Corner Radius (V4)", defaultValue=<rgba 0.125 0 0 0>}.R
   Outline Thickness=ScalarParameter() {defaultValue=5, parameterName="Outline Thickness (S)"}
  ) {materialFunction=&MF_UI_CoreUI_RoundedMask}
 ]
}
