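# Jam-system CommonActivatableWidget that shows a loop title and type on a bordered overlay, with a row for BPM, key and mode.
# SetTitleAndType copies TitleText into SelectedTitleText and LoopTypeText into the REMIX block, leaving SelectedLoopTypeText untouched.
# On Construct, a false IsPlayAndTransition collapses the song details row and REMIX block.
# This graph never sets the BPM, key or mode values.
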WidgetBlueprint /FMJamSystem/Widgets/WBP_JamLoop_TitleAndType
graphs [EventGraph {nodes [
   SetTitleAndType() {is=title K2Node_CustomEvent}
   > Set Text(Text=Get LoopTypeText() -> (LoopTypeText:Text), self=Get REMIX() -> (REMIX:Common Text Object Reference))
   > Set Text(Text=Get TitleText() -> (TitleText:Text), self=Get SelectedTitleText() -> (SelectedTitleText:Common Text Object Reference))
   Event Construct()
   > n1 = Branch(Condition=Get IsPlayAndTransition() -> (IsPlayAndTransition:bool))
   SetVisibility(execute=@n1.else, self=[Get SongDetailsBox() -> (SongDetailsBox:Horizontal Box Object Reference), Get REMIX()], InVisibility=Collapsed)
   Event PreConstruct()
   Event Tick()
  ]}]
variables [
 IsPlayAndTransition {type=boolean}
 LoopTypeText {type=string, default=""}
 TitleText {type=string, default=""}
]
widgets {
 info {inheritedWidgetCount=0, parentClass=&CommonActivatableWidget, rootWidgetClass=&Overlay, widgetCount=16}
 widgets [
  Overlay_PlayLoop {widgetClassPath=&Overlay, widget=Overlay_PlayLoop}
  BackgroundImage {
   widgetClassPath=&Image
   bIsVariable=true
   parent=Overlay_PlayLoop
   widget=BackgroundImage
   properties {
    visibility=HitTestInvisible
    brush {
     bIsDynamicallyLoaded=false
     drawAs=Image
     imageType=NoImage
     mirroring=NoMirror
     resourceName=None
     resourceObject=null
     tiling=NoTile
     imageSize=<100 50>
     margin=<ltrb 0 0 0 0>
     outlineSettings {
      bUseBrushTransparency=false
      roundingType=FixedRadius
      width=0
      color {colorUseRule=UseColor_Specified, specifiedColor=<rgba 0 0 0 0>}
      cornerRadii=<0 0 0 0.3>
     }
     tintColor {colorUseRule=UseColor_Specified, specifiedColor=<rgba 0 0 0 0.251>}
     uVRegion {bIsValid=false, max=<0 0>, min=<0 0>}
    }
   }
   slot {class=OverlaySlot, properties {horizontalAlignment=HAlign_Fill, verticalAlignment=VAlign_Fill, padding=<ltrb 5 5 5 5>}}
  }
  BorderImage {
   widgetClassPath=&Image
   bIsVariable=true
   parent=Overlay_PlayLoop
   widget=BorderImage
   properties {brush {
     bIsDynamicallyLoaded=false
     drawAs=Image
     imageType=NoImage
     mirroring=NoMirror
     resourceName=None
     resourceObject=/FMJamSystem/Materials/M_JamSystem_UI_DropLoopBorder
     tiling=NoTile
     imageSize=<32 32>
     margin=<ltrb 0 0 0 0>
     outlineSettings {
      bUseBrushTransparency=false
      roundingType=HalfHeightRadius
      width=0
      color {colorUseRule=UseColor_Specified, specifiedColor=<rgba 0 0 0 0>}
      cornerRadii=<0 0 0 0>
     }
     tintColor {colorUseRule=UseColor_Specified, specifiedColor=<rgba 1 1 1 1>}
     uVRegion {bIsValid=false, max=<0 0>, min=<0 0>}
    }}
   slot {class=OverlaySlot, properties {horizontalAlignment=HAlign_Fill, verticalAlignment=VAlign_Fill}}
  }
  VerticalBox_0 {
   widgetClassPath=&VerticalBox
   parent=Overlay_PlayLoop
   widget=VerticalBox_0
   slot {class=OverlaySlot, properties {horizontalAlignment=HAlign_Center, padding=<ltrb 5 10 5 10>}}
  }
  SelectedTitleText {
   widgetClassPath=&CommonTextBlock
   bIsVariable=true
   parent=VerticalBox_0
   widget=SelectedTitleText
   properties {
    justification=Center
    text="{Title}"
    font {
     bForceMonospaced=false
     bMaterialIsStencil=false
     fontMaterial=null
     fontObject=&BurbankSmall
     letterSpacing=0
     monospacedWidth=1
     size=30
     skewAmount=0
     typefaceFontName=Bold
     outlineSettings {
      bApplyOutlineToDropShadows=false
      bMiteredCorners=false
      bSeparateFillAlpha=false
      outlineMaterial=null
      outlineSize=0
      outlineColor=<rgba 0 0 0 1>
     }
    }
    shadowColorAndOpacity=<rgba 0.1562 0.1491 0.146 1>
   }
   slot {class=VerticalBoxSlot, properties {horizontalAlignment=HAlign_Center, verticalAlignment=VAlign_Center, padding=<ltrb 20 0 20 0>}}
  }
  Spacer_164 {widgetClassPath=&Spacer, parent=VerticalBox_0, widget=Spacer_164, properties {size=<1 30>}, slot {class=VerticalBoxSlot}}
  HorizontalBox_1 {
   widgetClassPath=&HorizontalBox
   parent=VerticalBox_0
   widget=HorizontalBox_1
   slot {class=VerticalBoxSlot, properties {horizontalAlignment=HAlign_Center, padding=<ltrb 5 0 5 0>}}
  }
  SelectedLoopTypeText {
   widgetClassPath=&CommonTextBlock
   bIsVariable=true
   parent=HorizontalBox_1
   widget=SelectedLoopTypeText
   properties {
    justification=Center
    text="{LoopType}"
    font {
     bForceMonospaced=false
     bMaterialIsStencil=false
     fontMaterial=null
     fontObject=&BurbankSmall
     letterSpacing=0
     monospacedWidth=1
     size=30
     skewAmount=0
     typefaceFontName=Bold
     outlineSettings {
      bApplyOutlineToDropShadows=false
      bMiteredCorners=false
      bSeparateFillAlpha=false
      outlineMaterial=null
      outlineSize=0
      outlineColor=<rgba 0 0 0 1>
     }
    }
    shadowColorAndOpacity=<rgba 0.1562 0.1491 0.146 1>
   }
   slot {class=HorizontalBoxSlot, properties {padding=<ltrb 20 0 0 0>}}
  }
  Spacer {widgetClassPath=&Spacer, parent=HorizontalBox_1, widget=Spacer, properties {size=<10 1>}, slot {class=HorizontalBoxSlot}}
  REMIX {
   widgetClassPath=&CommonTextBlock
   bIsVariable=true
   parent=HorizontalBox_1
   widget=REMIX
   properties {
    justification=Center
    text=REMIX
    font {
     bForceMonospaced=false
     bMaterialIsStencil=false
     fontMaterial=null
     fontObject=&BurbankSmall
     letterSpacing=0
     monospacedWidth=1
     size=30
     skewAmount=0
     typefaceFontName=Bold
     outlineSettings {
      bApplyOutlineToDropShadows=false
      bMiteredCorners=false
      bSeparateFillAlpha=false
      outlineMaterial=null
      outlineSize=0
      outlineColor=<rgba 0 0 0 1>
     }
    }
    shadowColorAndOpacity=<rgba 0.1562 0.1491 0.146 1>
   }
   slot {class=HorizontalBoxSlot, properties {padding=<ltrb 0 0 20 0>}}
  }
  SongDetailsBox {
   widgetClassPath=&HorizontalBox
   bIsVariable=true
   parent=VerticalBox_0
   widget=SongDetailsBox
   slot {class=VerticalBoxSlot, properties {horizontalAlignment=HAlign_Center, padding=<ltrb 0 10 0 0>}}
  }
  BPMText {
   widgetClassPath=&CommonTextBlock
   bIsVariable=true
   parent=SongDetailsBox
   widget=BPMText
   properties {
    text=Text Block
    font {
     bForceMonospaced=false
     bMaterialIsStencil=false
     fontMaterial=null
     fontObject=&BurbankSmall
     letterSpacing=0
     monospacedWidth=1
     size=30
     skewAmount=0
     typefaceFontName=Bold
     outlineSettings {
      bApplyOutlineToDropShadows=false
      bMiteredCorners=false
      bSeparateFillAlpha=false
      outlineMaterial=null
      outlineSize=0
      outlineColor=<rgba 0 0 0 1>
     }
    }
    shadowOffset=<0 0>
   }
   slot {class=HorizontalBoxSlot, properties {horizontalAlignment=HAlign_Left, verticalAlignment=VAlign_Center, padding=<ltrb 20 0 10 0>}}
  }
  BPM {
   widgetClassPath=&CommonTextBlock
   bIsVariable=true
   parent=SongDetailsBox
   widget=BPM
   properties {
    text=BPM
    font {
     bForceMonospaced=false
     bMaterialIsStencil=false
     fontMaterial=null
     fontObject=&BurbankSmall
     letterSpacing=0
     monospacedWidth=1
     size=30
     skewAmount=0
     typefaceFontName=Bold
     outlineSettings {
      bApplyOutlineToDropShadows=false
      bMiteredCorners=false
      bSeparateFillAlpha=false
      outlineMaterial=null
      outlineSize=0
      outlineColor=<rgba 0 0 0 1>
     }
    }
    shadowOffset=<0 0>
   }
   slot {class=HorizontalBoxSlot, properties {horizontalAlignment=HAlign_Left, verticalAlignment=VAlign_Center}}
  }
  Spacer_450 {widgetClassPath=&Spacer, parent=SongDetailsBox, widget=Spacer_450, properties {size=<20 1>}, slot {class=HorizontalBoxSlot}}
  KeyText {
   widgetClassPath=&CommonTextBlock
   bIsVariable=true
   parent=SongDetailsBox
   widget=KeyText
   properties {
    text=Text Block
    font {
     bForceMonospaced=false
     bMaterialIsStencil=false
     fontMaterial=null
     fontObject=&BurbankSmall
     letterSpacing=0
     monospacedWidth=1
     size=30
     skewAmount=0
     typefaceFontName=Bold
     outlineSettings {
      bApplyOutlineToDropShadows=false
      bMiteredCorners=false
      bSeparateFillAlpha=false
      outlineMaterial=null
      outlineSize=0
      outlineColor=<rgba 0 0 0 1>
     }
    }
    shadowOffset=<0 0>
   }
   slot {class=HorizontalBoxSlot, properties {horizontalAlignment=HAlign_Right, verticalAlignment=VAlign_Center, padding=<ltrb 0 0 10 0>}}
  }
  ModeText {
   widgetClassPath=&CommonTextBlock
   bIsVariable=true
   parent=SongDetailsBox
   widget=ModeText
   properties {
    text=Text Block
    font {
     bForceMonospaced=false
     bMaterialIsStencil=false
     fontMaterial=null
     fontObject=&BurbankSmall
     letterSpacing=0
     monospacedWidth=1
     size=30
     skewAmount=0
     typefaceFontName=Bold
     outlineSettings {
      bApplyOutlineToDropShadows=false
      bMiteredCorners=false
      bSeparateFillAlpha=false
      outlineMaterial=null
      outlineSize=0
      outlineColor=<rgba 0 0 0 1>
     }
    }
    shadowOffset=<0 0>
   }
   slot {class=HorizontalBoxSlot, properties {horizontalAlignment=HAlign_Right, verticalAlignment=VAlign_Center, padding=<ltrb 0 0 20 0>}}
  }
 ]
}
