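# WBP_JamPrompt is a Jam system input prompt widget pairing an action description with a keybind or touch icon.
# Hardware visibility borders show the keybind widget for gamepad or keyboard/mouse, and an icon switcher for touch.
# Legacy setup binds an action name and maps its identifier to an icon; Input Action setup uses the asset's description and fixed quick-join icon.
# Both setups first clear stored state without resetting visuals.
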WidgetBlueprint /FMJamSystem/Widgets/WBP_JamPrompt
graphs [
 EventGraph {nodes [
   n1 = Setup With Legacy Input!() -> (OutputDelegate, then, ActionName:Name, ActionDesc:Text, ActionIdentifier:Name) {is=title K2Node_CustomEvent}
   > Clear Internal State!(execute, self, UnsetVisualState:bool=false) -> (then)
   > n2 = Sequence!(execute) -> (then_0, then_1, then_2:Exec, then_3:Exec)
   > n3 = Set Action Name(Action Name:Name=@n1.ActionName) -> (Output_Get:Name=None)
   > SetBoundAction!(
    execute
    self:Fort Keybind Widget Object Reference=Get Keybind_Prompt() -> (Keybind_Prompt:Keybind Widget Object Reference)
    NewBoundAction:Name=@n3
   ) -> (then)
   Set ActionDesc(execute=@n2.then_1, ActionDesc:Text=@n1.ActionDesc) -> (Output_Get:Text)
   Set ActionIdentifier(execute=@n2.then_2, ActionIdentifier:Name=@n1.ActionIdentifier) -> (Output_Get:Name=None)
   SetActiveWidgetIndex(
    execute=@n2.then_3
    self=Get TouchIconSwitcher() -> (TouchIconSwitcher:Common Activatable Widget Switcher Object Reference)
    Index=Get Touch Icon Widget Index!(self, ActionIdentifier:"Name (by ref)"=Get ActionIdentifier() -> (ActionIdentifier:Name=None)) -> (SwitcherIndex:int=0) {is=title K2Node_CallFunction}
   )
   > SetText(self=Get RichText_Prompt() -> (RichText_Prompt:Common Rich Text Block Object Reference), InText=Get ActionDesc() -> (ActionDesc:Text))
   n4 = Clear Internal State!() -> (OutputDelegate, then, UnsetVisualState:bool=true) {is=title K2Node_CustomEvent}
   > n5 = Sequence!(execute) -> (then_0, then_1)
   > Set Action Name() -> (Output_Get=None)
   > Set ActionDesc()
   > Set ActionIdentifier() -> (Output_Get=None)
   > Set Input Action(Input Action:Input Action Object Reference) -> (Output_Get:Input Action Object Reference)
   Branch(execute=@n5.then_1, Condition=@n4.UnsetVisualState)
   n6 = Get Keybind_Prompt()
   > SetBoundInputAction!(execute, self:Fort Keybind Widget Object Reference=@n6, NewAction:Input Action Object Reference) -> (then)
   > SetBoundAction(self=@n6, NewBoundAction=None)
   > Set Text(self:Rich Text Block Object Reference=Get RichText_Prompt())
   > Clear(TargetArray:Array of Names=Get BoundActionFallbacks!(self:Fort Keybind Widget Object Reference=Get Keybind_Prompt()) -> (BoundActionFallbacks:Array of Names))
   n7 = Setup With Input Action!() -> (OutputDelegate, then, Input Action:Input Action Object Reference) {is=title K2Node_CustomEvent}
   > Clear Internal State(UnsetVisualState=false) {is=title K2Node_CallFunction}
   > Set Input Action(Input Action=@n7.Input Action)
   > SetBoundInputAction(self=Get Keybind_Prompt(), NewAction=Get Input Action() -> (Input Action:Input Action Object Reference))
   > Set ActionDesc(ActionDesc=Get ActionDescription!(self:Input Action Object Reference=Get Input Action()) -> (ActionDescription:Text))
   > SetActiveWidgetIndex(self=Get TouchIconSwitcher(), Index=3)
   > SetText(self=Get RichText_Prompt(), InText=Get ActionDesc())
  ]}
 Get Touch Icon Widget Index {nodes [
   n1 = Get Touch Icon Widget Index!() -> (then, ActionIdentifier:"Name (by ref)") {is=title K2Node_FunctionEntry}
   > n2 = Switch on Name!(execute:Exec, Selection:Name=@n1) -> (Default:Exec, None:Exec, ChangeLoop:Exec, PlayLoop:Exec, PlayAndTransition:Exec, StopLoop:Exec, GlobalControls:Exec) {is=title K2Node_SwitchName, pinNames [None, ChangeLoop, PlayLoop, PlayAndTransition, StopLoop, GlobalControls]}
   Return Node!(execute=[@n2.ChangeLoop, @n2.None, @n2.PlayLoop], SwitcherIndex:int=1)
   > Return Node(SwitcherIndex=0)
   Return Node(execute=[@n2.GlobalControls, @n2.PlayAndTransition], SwitcherIndex=3)
   Return Node(execute=@n2.StopLoop, SwitcherIndex=2)
  ]}
 Is Keybind Already Set {nodes [
   Is Keybind Already Set() {is=title K2Node_FunctionEntry}
   > n1 = Sequence!(execute) -> (then_0, then_1, then_2:Exec, then_3:Exec)
   Return Node!(execute=@n1.then_3, ReturnValue:bool=false)
   n2 = Get BoundAction!(self:Fort Keybind Widget Object Reference=Get Keybind_Prompt()) -> (BoundAction:Name=None)
   > Branch(Condition=AND Boolean(A="Not Equal (Name)"(A=@n2) -> (ReturnValue=~), B="Not Equal (Name)"(A=@n2) -> (ReturnValue=~)))
   > Return Node(ReturnValue=true)
   n3 = Get Keybind_Prompt()
   Branch(
    execute=@n1.then_1
    Condition=IsValid(Object=Get BoundInputAction!(self:Fort Keybind Widget Object Reference=@n3) -> (BoundInputAction:Input Action Object Reference))
   )
   > Return Node(ReturnValue=true)
   n4 = For Each Loop(Exec=@n1.then_2, Array:Array of Names=Get BoundActionFallbacks(self=Get Keybind_Prompt())) -> (Array Element:Name)
   > Branch(Condition=AND Boolean(A="Not Equal (Name)"(A=@n4) -> (ReturnValue=~), B="Not Equal (Name)"(A=@n4) -> (ReturnValue=~)))
   > Return Node(ReturnValue=true)
   Branch(Condition="Not Equal (Name)"(A=Get Action Name() -> (Action Name:Name=None)) -> (ReturnValue=~))
  ]}
]
variables [
 Action Name {type=string, default=None}
 ActionDesc {type=string, default=""}
 ActionIdentifier {type=string, default=None}
 Input Action {type=InputAction, default=null}
]
widgets {
 info {inheritedWidgetCount=0, parentClass=&UserWidget, rootWidgetClass=&Overlay, widgetCount=12}
 widgets [
  Overlay_20 {widgetClassPath=&Overlay, widget=Overlay_20}
  PromptLine {widgetClassPath=&HorizontalBox, parent=Overlay_20, widget=PromptLine, slot {class=OverlaySlot}}
  CommonHardwareVisibilityBorder_89 {
   widgetClassPath=&CommonHardwareVisibilityBorder
   parent=PromptLine
   widget=CommonHardwareVisibilityBorder_89
   properties {visibilityQuery {
     autoDescription=" ANY( Input.Gamepad, Input.MouseAndKeyboard )"
     tokenStreamVersion=0
     userDescription=Is using KBM or Gamepad
     queryTokenStream [0,1,1,2,0,1]
     tagDictionary [{tagName=Input.Gamepad}, {tagName=Input.MouseAndKeyboard}]
    }}
   slot {class=HorizontalBoxSlot, properties {verticalAlignment=VAlign_Center, padding=<ltrb 10 10 10 10>}}
  }
  Keybind_Prompt {
   widgetClassPath=/FortUILibrary/Input/KeybindWidget
   bIsVariable=true
   parent=CommonHardwareVisibilityBorder_89
   widget=Keybind_Prompt
   properties {bShowUnboundStatus=true, bUseKeyboardTextureForKeys=true, boundKeyFallback=Enter}
   slot {class=BorderSlot, properties {padding=<ltrb 0 0 0 0>}}
  }
  CommonHardwareVisibilityBorder_90 {
   widgetClassPath=&CommonHardwareVisibilityBorder
   parent=PromptLine
   widget=CommonHardwareVisibilityBorder_90
   properties {visibilityQuery {
     autoDescription=" ANY( Input.Touch )"
     tokenStreamVersion=0
     userDescription=Is Using Touch Input
     queryTokenStream [0,1,1,1,0]
     tagDictionary [{tagName=Input.Touch}]
    }}
   slot {class=HorizontalBoxSlot, properties {verticalAlignment=VAlign_Center, padding=<ltrb 10 10 10 10>}}
  }
  TouchIconSwitcher {
   widgetClassPath=&CommonActivatableWidgetSwitcher
   bIsVariable=true
   parent=CommonHardwareVisibilityBorder_90
   widget=TouchIconSwitcher
   properties {transitionCurveType=Linear, transitionDuration=0}
   slot {class=BorderSlot, properties {padding=<ltrb 0 0 0 0>}}
  }
  TouchKeybindIcon_PlaceAutoJammer {
   widgetClassPath=&Image
   bIsVariable=true
   parent=TouchIconSwitcher
   widget=TouchKeybindIcon_PlaceAutoJammer
   properties {
    bOverride_Cursor=true
    visibility=SelfHitTestInvisible
    brush {
     bIsDynamicallyLoaded=false
     drawAs=Image
     imageType=FullColor
     mirroring=NoMirror
     resourceName=None
     resourceObject=/FMCore/Textures/T_HoloJam_Icon_add
     tiling=NoTile
     imageSize=<58 58>
     margin=<ltrb 0 0 0 0>
     outlineSettings {
      bUseBrushTransparency=false
      roundingType=HalfHeightRadius
      width=0
      color {colorUseRule=UseColor_Specified, specifiedColor=<rgba 0 0 0 0>}
      cornerRadii=<0 0 0 0>
     }
     tintColor {colorUseRule=UseColor_Specified, specifiedColor=<rgba 1 1 1 1>}
     uVRegion {bIsValid=false, max=<0 0>, min=<0 0>}
    }
   }
   slot {class=WidgetSwitcherSlot}
  }
  TouchKeybindIcon_ChangeAutoJammerSong {
   widgetClassPath=&Image
   bIsVariable=true
   parent=TouchIconSwitcher
   widget=TouchKeybindIcon_ChangeAutoJammerSong
   properties {
    bOverride_Cursor=true
    visibility=SelfHitTestInvisible
    brush {
     bIsDynamicallyLoaded=false
     drawAs=Image
     imageType=FullColor
     mirroring=NoMirror
     resourceName=None
     resourceObject=/FMCore/Textures/T_HoloJam_Icon_change
     tiling=NoTile
     imageSize=<58 58>
     margin=<ltrb 0 0 0 0>
     outlineSettings {
      bUseBrushTransparency=false
      roundingType=HalfHeightRadius
      width=0
      color {colorUseRule=UseColor_Specified, specifiedColor=<rgba 0 0 0 0>}
      cornerRadii=<0 0 0 0>
     }
     tintColor {colorUseRule=UseColor_Specified, specifiedColor=<rgba 1 1 1 1>}
     uVRegion {bIsValid=false, max=<0 0>, min=<0 0>}
    }
   }
   slot {class=WidgetSwitcherSlot}
  }
  TouchKeybindIcon_StopAutoJammerSong {
   widgetClassPath=&Image
   bIsVariable=true
   parent=TouchIconSwitcher
   widget=TouchKeybindIcon_StopAutoJammerSong
   properties {
    bOverride_Cursor=true
    visibility=SelfHitTestInvisible
    brush {
     bIsDynamicallyLoaded=false
     drawAs=Image
     imageType=FullColor
     mirroring=NoMirror
     resourceName=None
     resourceObject=/FMCore/Textures/T_HoloJam_Icon_cancel
     tiling=NoTile
     imageSize=<58 58>
     margin=<ltrb 0 0 0 0>
     outlineSettings {
      bUseBrushTransparency=false
      roundingType=HalfHeightRadius
      width=0
      color {colorUseRule=UseColor_Specified, specifiedColor=<rgba 0 0 0 0>}
      cornerRadii=<0 0 0 0>
     }
     tintColor {colorUseRule=UseColor_Specified, specifiedColor=<rgba 1 1 1 1>}
     uVRegion {bIsValid=false, max=<0 0>, min=<0 0>}
    }
   }
   slot {class=WidgetSwitcherSlot}
  }
  TouchKeybindIcon_QuickJoin {
   widgetClassPath=&Image
   bIsVariable=true
   parent=TouchIconSwitcher
   widget=TouchKeybindIcon_QuickJoin
   properties {
    bOverride_Cursor=true
    visibility=SelfHitTestInvisible
    brush {
     bIsDynamicallyLoaded=false
     drawAs=Image
     imageType=FullColor
     mirroring=NoMirror
     resourceName=None
     resourceObject=/FMJamSystem/Textures/Icons/T_Icon_Sparks_JamTrack
     tiling=NoTile
     imageSize=<58 58>
     margin=<ltrb 0 0 0 0>
     outlineSettings {
      bUseBrushTransparency=false
      roundingType=HalfHeightRadius
      width=0
      color {colorUseRule=UseColor_Specified, specifiedColor=<rgba 0 0 0 0>}
      cornerRadii=<0 0 0 0>
     }
     tintColor {colorUseRule=UseColor_Specified, specifiedColor=<rgba 1 1 1 1>}
     uVRegion {bIsValid=false, max=<0 0>, min=<0 0>}
    }
   }
   slot {class=WidgetSwitcherSlot}
  }
  TextAlignmentCanvasPanel {
   widgetClassPath=&CanvasPanel
   parent=PromptLine
   widget=TextAlignmentCanvasPanel
   slot {class=HorizontalBoxSlot, properties {verticalAlignment=VAlign_Center}}
  }
  RichText_Prompt {
   widgetClassPath=&CommonRichTextBlock
   bIsVariable=true
   parent=TextAlignmentCanvasPanel
   widget=RichText_Prompt
   properties {
    bOverrideDefaultStyle=true
    inlineIconDisplayMode=TextOnly
    text="{ContextualAction}"
    defaultTextStyleOverride {
     overflowPolicy=Clip
     transformPolicy=None
     colorAndOpacity {colorUseRule=UseColor_Specified, specifiedColor=<rgba 1 1 1 1>}
     font {
      bForceMonospaced=false
      bMaterialIsStencil=false
      fontMaterial=null
      fontObject=&BurbankSmall
      letterSpacing=0
      monospacedWidth=1
      size=21
      skewAmount=0
      typefaceFontName=Bold
      outlineSettings {
       bApplyOutlineToDropShadows=true
       bMiteredCorners=false
       bSeparateFillAlpha=true
       outlineMaterial=&M-FontShadow
       outlineSize=3
       outlineColor=<rgba 0 0 0 0.498>
      }
     }
     highlightColor {colorUseRule=UseColor_Specified, specifiedColor=<rgba 0 0 0 1>}
     highlightShape {
      bIsDynamicallyLoaded=false
      drawAs=Image
      imageType=NoImage
      mirroring=NoMirror
      resourceName=None
      resourceObject=null
      tiling=NoTile
      imageSize=<32 32>
      margin=<ltrb 0 0 0 0>
      outlineSettings {
       bUseBrushTransparency=false
       roundingType=HalfHeightRadius
       width=0
       color {colorUseRule=UseColor_Specified, specifiedColor=<rgba 0 0 0 0>}
       cornerRadii=<0 0 0 0>
      }
      tintColor {colorUseRule=UseColor_Specified, specifiedColor=<rgba 1 1 1 1>}
      uVRegion {bIsValid=false, max=<0 0>, min=<0 0>}
     }
     selectedBackgroundColor {colorUseRule=UseColor_Specified, specifiedColor=<rgba 1 0.7343 0.4192 1>}
     shadowColorAndOpacity=<rgba 0.0417 0.0398 0.0389 1>
     shadowOffset=<1 1>
     strikeBrush {
      bIsDynamicallyLoaded=false
      drawAs=Image
      imageType=NoImage
      mirroring=NoMirror
      resourceName=None
      resourceObject=null
      tiling=NoTile
      imageSize=<32 32>
      margin=<ltrb 0 0 0 0>
      outlineSettings {
       bUseBrushTransparency=false
       roundingType=HalfHeightRadius
       width=0
       color {colorUseRule=UseColor_Specified, specifiedColor=<rgba 0 0 0 0>}
       cornerRadii=<0 0 0 0>
      }
      tintColor {colorUseRule=UseColor_Specified, specifiedColor=<rgba 1 1 1 1>}
      uVRegion {bIsValid=false, max=<0 0>, min=<0 0>}
     }
     underlineBrush {
      bIsDynamicallyLoaded=false
      drawAs=Image
      imageType=NoImage
      mirroring=NoMirror
      resourceName=None
      resourceObject=null
      tiling=NoTile
      imageSize=<32 32>
      margin=<ltrb 0 0 0 0>
      outlineSettings {
       bUseBrushTransparency=false
       roundingType=HalfHeightRadius
       width=0
       color {colorUseRule=UseColor_Specified, specifiedColor=<rgba 0 0 0 0>}
       cornerRadii=<0 0 0 0>
      }
      tintColor {colorUseRule=UseColor_Specified, specifiedColor=<rgba 1 1 1 1>}
      uVRegion {bIsValid=false, max=<0 0>, min=<0 0>}
     }
    }
   }
   slot {
    class=CanvasPanelSlot
    properties {bAutoSize=true, layoutData {alignment=<0 0.5>, anchors {maximum=<0 0.5>, minimum=<0 0.5>}, offsets=<ltrb 0 0 100 30>}}
   }
  }
 ]
}
