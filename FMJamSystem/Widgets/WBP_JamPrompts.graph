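# WBP_JamPrompts is a HUD widget that lists input prompts for the Jam system.
# It filters input aliases by context object, always allowing StopLoop, and reuses pooled WBP_JamPrompt children, collapsing extras.
# Each prompt uses an Enhanced Input action or falls back to legacy binding data.
# It hides for touch input or active text chat, and in Campaign the Jamming tag toggles certain HUD elements and hold-to-craft ammo.
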WidgetBlueprint /FMJamSystem/Widgets/WBP_JamPrompts
graphs [
 Apply Alias LegacyBindingData To JamPrompt Widget {nodes [
   Apply Alias LegacyBindingData To JamPrompt Widget!() -> (then, Alias:Input Alias Object Reference, JamPrompt:WBP Jam Prompt Object Reference) {is=title K2Node_FunctionEntry}
   > n1 = Sequence()
   > n2 = GetLegacyBindingData!(execute, self:Input Alias Object Reference=Get Alias() -> (Alias:Input Alias Object Reference)) -> (
    then
    OutLegacyBinding_SpecializedActionName:Name
    OutLegacyBinding_GeneralActionName:Name
    OutLegacyBinding_ActionIdentifierName:Name
    OutLegacyBinding_ActionDescription:Text
   )
   > n3 = Sequence!(execute) -> (then_0, then_1, then_2:Exec, then_3:Exec)
   > Set SpecializedAction(SpecializedAction:Name=@n2) -> (Output_Get:Name=None)
   Set GeneralizedAction(execute=@n3.then_1, GeneralizedAction:Name=@n2.OutLegacyBinding_GeneralActionName) -> (Output_Get:Name=None)
   Set HACK_IdentifierName(execute=@n3.then_2, HACK_IdentifierName:Name=@n2.OutLegacyBinding_ActionIdentifierName) -> (Output_Get:Name=None)
   Set ActionDescription(execute=@n3.then_3, ActionDescription:Text=@n2.OutLegacyBinding_ActionDescription) -> (Output_Get:Text)
   n4 = Sequence!(execute=@n1.then_1) -> (then_0, then_1)
   Setup With Legacy Input!(
    execute=@n4.then_1
    self:WBP Jam Prompt Object Reference=Get JamPrompt() -> (JamPrompt:WBP Jam Prompt Object Reference)
    ActionName:Name=Get SpecializedAction() -> (SpecializedAction:Name=None)
    ActionDesc:Text=Get ActionDescription() -> (ActionDescription:Text)
    ActionIdentifier:Name=Get HACK_IdentifierName() -> (HACK_IdentifierName:Name=None)
   ) -> (then)
   n5 = Get BoundActionFallbacks!(self:Fort Keybind Widget Object Reference=Get Keybind_Prompt!(self:WBP Jam Prompt Object Reference=Get JamPrompt()) -> (Keybind_Prompt:Keybind Widget Object Reference)) -> (BoundActionFallbacks:Array of Names)
   Clear(execute=@n4.then_0, TargetArray:Array of Names=@n5)
   > Add(TargetArray:Array of Names=@n5, NewItem:"Name (by ref)"=Get GeneralizedAction() -> (GeneralizedAction:Name=None)) {is=title K2Node_CallArrayFunction}
  ]}
 CanShowWithoutContext {nodes [
   n1 = CanShowWithoutContext!() -> (then, InputAlias:Input Alias Object Reference) {is=title K2Node_FunctionEntry}
   > n2 = GetLegacyBindingData!(execute, self=@n1) -> (then, OutLegacyBinding:Legacy Input Binding Alias Data Structure)
   > Return Node!(
    execute
    ReturnValue:bool="Equal (Name)"(
     A=Break Legacy Input Binding Alias Data!(LegacyInputBindingAliasData:"Legacy Input Binding Alias Data Structure (by ref)"=@n2) -> (SpecializedActionName:Name, GeneralActionName:Name, ActionIdentifierName:Name, ActionDescription:Text) {is=title K2Node_BreakStruct}.ActionIdentifierName
     B=StopLoop
    )
   )
  ]}
 EventGraph {nodes [
   n1 = On Input Method Changed!() -> (OutputDelegate, then, NewInputType:ECommonInputType Enum) {is=title K2Node_CustomEvent}
   n2 = NOT Boolean(A=Get Keep Visible While Using Touch Input() -> (Keep Visible While Using Touch Input:bool))
   n3 = "Equal (Enum)"(A=@n1.NewInputType, B=Touch)
   > SetVisibility(
    self=Get FortSimpleWidgetAnimationsPanel() -> (FortSimpleWidgetAnimationsPanel:Fort Simple Widget Animations Panel Object Reference)
    InVisibility=Select(Option 0=SelfHitTestInvisible, Option 1=Hidden, Index=AND Boolean(A=@n3, B=@n2))
   )
   SetupTextChatListeners() {is=title K2Node_CustomEvent}
   > n4 = StartListeningToEvent!(
    execute:Exec
    Target:Gameplay Event Router Component Object Reference=GetGlobalEventRouter()
    EventType:Script Struct Object Reference=/FortniteCoreUI/UI/Social/Chat/TextChat/BPS_TextChatActiveStateChange
    Context:Object Reference=GetOwningPlayer()
    EventBubblingRule:EEventBubblingRule Enum
   ) -> (
    then:Exec
    AsyncTaskProxy:Async Action Start Listening To Event Object Reference
    OnEventReceived:Exec
    Payload:BPS Text Chat Active State Change Structure
   )
   > Set TextChatActiveListener(TextChatActiveListener:Async Action Start Listening To Event Object Reference=@n4) -> (Output_Get:Async Action Start Listening To Event Object Reference)
   SetVisibility(
    execute=@n4.OnEventReceived
    InVisibility=Select(
     Option 0=HitTestInvisible
     Option 1=Hidden
     Index=Break BPS Text Chat Active State Change!(BPS_TextChatActiveStateChange:"BPS Text Chat Active State Change Structure (by ref)"=@n4.Payload) -> (bActive_4_#1:bool=~) {is=title K2Node_BreakStruct}
    )
   )
   n5 = Event PreConstruct()
   > Parent: PreConstruct(IsDesignTime=@n5.IsDesignTime)
   Event Construct()
   > Parent: Construct()
   > n6 = Sequence()
   n7 = Get CommonInputSubsystem!(PlayerController:Player Controller Object Reference=GetOwningPlayer()) -> (ReturnValue:Common Input Object Reference) {is=title K2Node_GetSubsystemFromPC}
   > Bind Event to On Input Method Changed(self=@n7, Delegate=Create Event())
   > On Input Method Changed!(execute, self, NewInputType:ECommonInputType Enum=GetCurrentInputType(self=@n7)) -> (then) {is=title K2Node_CallFunction}
   > SetupTextChatListeners() {is=title K2Node_CallFunction}
   Branch(execute=@n6.then_1, Condition=IsInSubGame!(SubGame:ESubGame Enum=Campaign) -> (ReturnValue:bool))
   n8 = Get PlayerState!(self:Controller Object Reference=GetOwningPlayer()) -> (PlayerState:Player State Object Reference)
   > n9 = WaitGameplayTagAddToActor!(
    execute:Exec
    TargetActor:Actor Object Reference=@n8
    Tag:Gameplay Tag Structure=(TagName="Gameplay.Action.Player.Jam.Jamming")
    OnlyTriggerOnce:bool=false
   ) -> (then:Exec, AsyncTaskProxy:Ability Async Wait Gameplay Tag Added Object Reference, Added:Exec) {is=title K2Node_AsyncAction}
   > n10 = WaitGameplayTagRemoveFromActor!(
    execute:Exec
    TargetActor:Actor Object Reference=@n8
    Tag:Gameplay Tag Structure=(TagName="Gameplay.Action.Player.Jam.Jamming")
    OnlyTriggerOnce:bool=false
   ) -> (then:Exec, AsyncTaskProxy:Ability Async Wait Gameplay Tag Removed Object Reference, Removed:Exec) {is=title K2Node_AsyncAction}
   ModifyUIForSTW!(execute=@n10.Removed, self, IsScreenActive:bool=false) -> (then)
   ModifyUIForSTW(execute=@n9.Added, IsScreenActive=true)
   Event Destruct()
   > StopListeningToEvent(self:Async Action Start Listening To Event Object Reference=Get TextChatActiveListener() -> (TextChatActiveListener:Async Action Start Listening To Event Object Reference))
  ]}
 ModifyUIForSTW {nodes [
   n1 = ModifyUIForSTW!() -> (then, IsScreenActive:bool=~) {is=title K2Node_FunctionEntry}
   > SetHUDElementVisibility!(
    execute
    self:Fort HUDContext Object Reference=Get FortHUDContext!() -> (ReturnValue:Fort HUDContext Object Reference) {is=title K2Node_GetBlueprintContextEx}
    HUDElementTags:"Gameplay Tag Container Structure (by ref)"=MakeLiteralGameplayTagContainer!(Value:Gameplay Tag Container Structure=(GameplayTags=((TagName="HUD.TopRight.MissionTracker"),(TagName="HUD.Quickbar.Build.EquippedItem")))) -> (ReturnValue:Gameplay Tag Container Structure)
    bHideElements:bool=@n1
    ReasonName:"Name (by ref)"=None
   ) -> (then)
   > n2 = Cast To FortPlayerController(Object=GetOwningPlayer())
   > SetAllowHoldForAmmoCrafting!(execute, self:Fort Player Controller Object Reference=@n2, bAllow:bool=NOT Boolean(A=@n1)) -> (then)
  ]}
 Receive New Input Aliases {nodes [
   Receive New Input Aliases!() -> (then, New Aliases:Array of Input Alias Object References) {is=title K2Node_FunctionEntry}
   > n1 = Sequence()
   Set Current Child Index(execute=@n1.then_1, Current Child Index:int=0) -> (Output_Get:int=0)
   > n2 = For Each Loop(Array:Array of Input Alias Object References=Get New Aliases() -> (New Aliases:Array of Input Alias Object References)) -> (Array Element:Input Alias Object Reference)
   n3 = Get PromptContextObject() -> (PromptContextObject:Object Reference)
   n4 = Get ContextObject!(self:Input Alias Object Reference=@n2) -> (ContextObject:Object Reference)
   > Branch(Condition=OR Boolean(
     A="Equal (Object)"(A=@n4, B=@n3)
     B=CanShowWithoutContext!(InputAlias:Input Alias Object Reference=@n2) -> (ReturnValue:bool) {is=title K2Node_CallFunction}
    ))
   > Set Current Alias(Current Alias:Input Alias Object Reference=@n2) -> (Output_Get:Input Alias Object Reference)
   > n5 = Sequence!(execute) -> (then_0, then_1, then_2:Exec, then_3:Exec)
   n6 = GetChildrenCount!(self:Panel Widget Object Reference=Get MVVMHost() -> (MVVMHost:Vertical Box Object Reference)) -> (ReturnValue:int=0)
   > n7 = Branch(Condition="integer >= integer"(A=int - int(A=@n6), B=Get Current Child Index() -> (Current Child Index:int=0)))
   > n8 = Cast To WBP_JamPrompt!(
    execute
    Object=GetChildAt!(self:Panel Widget Object Reference=Get MVVMHost(), Index:int=Get Current Child Index() -> (Current Child Index=0)) -> (ReturnValue:Widget Object Reference)
   ) -> (then, CastFailed, AsWBP Jam Prompt:WBP Jam Prompt Object Reference) {is=title K2Node_DynamicCast}
   > Set Current Widget(Current Widget:WBP Jam Prompt Object Reference=@n8) -> (Output_Get:WBP Jam Prompt Object Reference)
   n9 = Create Widget!(
    execute:Exec=[@n7.else, @n8.CastFailed]
    Class:User Widget Class Reference=&/WBP_JamPrompt.WBP_JamPrompt_C
    OwningPlayer:Player Controller Object Reference=GetOwningPlayer()
   ) -> (then:Exec, ReturnValue:WBP Jam Prompt Object Reference)
   > n10 = Set Current Widget(Current Widget=@n9)
   > AddChild(self=Get MVVMHost(), Content=@n10)
   n11 = Get!(execute:Exec=@n5.then_2, self:Input Alias Object Reference=Get Current Alias() -> (Current Alias:Input Alias Object Reference)) -> (then:Exec, else:Exec, EnhancedInputAction:Input Action Object Reference)
   > Setup With Input Action!(
    execute
    self:WBP Jam Prompt Object Reference=Get Current Widget() -> (Current Widget:WBP Jam Prompt Object Reference)
    Input Action:Input Action Object Reference=@n11
   ) -> (then)
   Apply Alias LegacyBindingData To JamPrompt Widget!(
    execute=@n11.else
    self
    Alias:Input Alias Object Reference=Get Current Alias()
    JamPrompt:WBP Jam Prompt Object Reference=Get Current Widget()
   ) -> (then) {is=title K2Node_CallFunction}
   Increment Int(" "=@n5.then_3, Value=Get Current Child Index() -> (Current Child Index=0))
   n12 = Get Current Widget()
   Clear Internal State!(execute=@n5.then_1, self:WBP Jam Prompt Object Reference=@n12, UnsetVisualState:bool=true) -> (then)
   > SetVisibility(self=@n12, InVisibility=HitTestInvisible)
   n13 = Get MVVMHost()
   n14 = For Loop(
    execute=@n2.Completed
    FirstIndex=Get Current Child Index() -> (Current Child Index=0)
    LastIndex=int - int(A=GetChildrenCount(self=@n13) -> (ReturnValue=0))
   )
   n15 = GetChildAt(self=Get MVVMHost(), Index=@n14)
   > Is Valid(InputObject=@n15)
   > SetVisibility(self=@n15, InVisibility=Collapsed)
   n16 = Get MVVMHost()
   n17 = Set Original MVVM Host Child Last Index(execute=@n1.then_0, Original MVVM Host Child Last Index:int=int - int(A=GetChildrenCount(self=@n16) -> (ReturnValue=0))) -> (Output_Get:int=0)
   > n18 = For Loop(LastIndex=@n17)
   n19 = Reverse for Each Loop!(
    Exec:Exec=@n18.Completed
    Array:Array of Integers=Get MVVM Host Child Indices To Remove() -> (MVVM Host Child Indices To Remove:Array of Integers)
   ) -> (LoopBody:Exec, ArrayIndex:int, ArrayElement:int, Completed:Exec) {is=title K2Node_MacroInstance}
   > RemoveChildAt!(execute, self:Panel Widget Object Reference=Get MVVMHost(), Index:int=@n19.ArrayElement) -> (then, ReturnValue:bool)
   n20 = Cast To WBP_JamPrompt(execute=@n18.LoopBody, Object=GetChildAt(self=Get MVVMHost(), Index=@n18))
   Add(execute=@n20.CastFailed, TargetArray:Array of Integers=Get MVVM Host Child Indices To Remove(), NewItem:"Integer (by ref)"=@n18)
  ]}
 UpdatePromptContext {nodes [
   n1 = UpdatePromptContext!() -> (then, PromptContextObject:Object Reference) {is=title K2Node_FunctionEntry}
   > Set PromptContextObject(PromptContextObject:Object Reference=@n1) -> (Output_Get:Object Reference)
   > n2 = Get!(execute) -> (then, else, ManagedInputPromptsVM:Managed Input Prompts VM Object Reference)
   > Receive New Input Aliases!(
    execute
    self
    New Aliases:Array of Input Alias Object References=Get AllKnownInputAliases!(self:Managed Input Prompts VM Object Reference=@n2) -> (AllKnownInputAliases:Array of Input Alias Object References)
   ) -> (then) {is=title K2Node_CallFunction}
  ]}
]
variables [
 bInvertJamBasedVisibility {type=boolean}
 bShowOnlyWhileJamming {type=boolean, default=true}
 Keep Visible While Using Touch Input {type=boolean}
 PromptContextObject {type=Object, default=null}
 TextChatActiveListener {type=AsyncAction_StartListeningToEvent, default=null}
]
widgets {
 info {
  inheritedWidgetCount=0
  parentClass=&CommonUserWidget
  rootWidgetClass=/Script/FortniteUI.FortSimpleWidgetAnimationsPanel
  widgetCount=6
 }
 widgets [
  FortSimpleWidgetAnimationsPanel {
   widgetClassPath=&FortSimpleWidgetAnimationsPanel
   bIsVariable=true
   widget=FortSimpleWidgetAnimationsPanel
   properties {animations [
     Intro {
      bIsEnabled=true
      onAnimationsFinished=()
      targets [{
        alpha=/Game/Athena/HUD/HUD_ContextualControls_IntroAlpha
        bResetOnFinish=false
        bShouldAnimateAlpha=true
        bShouldAnimateScale=false
        bShouldAnimateTranslation=false
        scale=null
        targetName=HorizontalBox_2
        translation=null
        materialScalarAnimations []
       }]
     }
     Outro {
      bIsEnabled=true
      onAnimationsFinished=()
      targets [{
        alpha=/Game/Athena/HUD/HUD_ContextualControls_OutroAlpha
        bResetOnFinish=false
        bShouldAnimateAlpha=true
        bShouldAnimateScale=false
        bShouldAnimateTranslation=false
        scale=null
        targetName=HorizontalBox_2
        translation=null
        materialScalarAnimations []
       }]
     }
    ]}
  }
  HorizontalBox_2 {
   widgetClassPath=&HorizontalBox
   parent=FortSimpleWidgetAnimationsPanel
   widget=HorizontalBox_2
   slot {class=FortSimpleWidgetAnimationsSlot}
  }
  ActionsInvalidationBox {
   widgetClassPath=/Script/UMG.InvalidationBox
   parent=HorizontalBox_2
   widget=ActionsInvalidationBox
   properties {bCanCache=false, renderTransformPivot=<0 1>}
   slot {class=HorizontalBoxSlot}
  }
  ScaleBox_1 {
   widgetClassPath=&ScaleBox
   parent=ActionsInvalidationBox
   widget=ScaleBox_1
   properties {stretch=UserSpecified, userSpecifiedScale=0.8}
   slot {class=PanelSlot}
  }
  ControllerCalloutBox {widgetClassPath=&VerticalBox, bIsVariable=true, parent=ScaleBox_1, widget=ControllerCalloutBox, slot {class=ScaleBoxSlot}}
  MVVMHost {widgetClassPath=&VerticalBox, bIsVariable=true, parent=ControllerCalloutBox, widget=MVVMHost, slot {class=VerticalBoxSlot}}
 ]
}
