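# Persistent variant of the WBP_JamPrompts widget that hides itself while the owning player carries the Pawn.HidePlayerPawn.HidingProp gameplay tag.
# On Construct it starts an async tag-count listener on the owning player controller and stores the task.
# Each count change sets visibility to Hidden when the count is above zero, otherwise HitTestInvisible.
# Destruct cancels the listener; the widget adds no widgets of its own.
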WidgetBlueprint /FMJamSystem/Widgets/WBP_JamPrompts_Persistant
graphs [EventGraph {nodes [
   Event Construct()
   > Parent: Construct()
   > n1 = WaitGameplayTagCountChangedOnActor!(execute:Exec, TargetActor:Actor Object Reference=GetOwningPlayer(), Tag:Gameplay Tag Structure=(TagName="Pawn.HidePlayerPawn.HidingProp")) -> (then:Exec, AsyncTaskProxy:Ability Async Wait Gameplay Tag Count Changed Object Reference, TagCountChanged:Exec, TagCount:int) {is=title K2Node_AsyncAction}
   > Set HidingPropTagCheck(HidingPropTagCheck:Ability Async Wait Gameplay Tag Count Changed Object Reference=@n1) -> (Output_Get:Ability Async Wait Gameplay Tag Count Changed Object Reference)
   SetVisibility(
    execute=@n1.TagCountChanged
    self=Self-Reference()
    InVisibility=Select(Option 0=HitTestInvisible, Option 1=Hidden, Index="integer > integer"(A=@n1.TagCount))
   )
   Event Destruct()
   > Cancel(self:Cancellable Async Action Object Reference=Get HidingPropTagCheck() -> (HidingPropTagCheck:Ability Async Wait Gameplay Tag Count Changed Object Reference))
   > Parent: Destruct()
  ]}]
variables [HidingPropTagCheck {type=AbilityAsync_WaitGameplayTagCountChanged, default=null}]
widgets {
 info {inheritedWidgetCount=0, parentClass=&/WBP_JamPrompts.WBP_JamPrompts_C, rootWidgetClass=null, widgetCount=0}
 widgets []
}
