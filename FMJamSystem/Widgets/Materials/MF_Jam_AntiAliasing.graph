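# DelMarUI material function that anti-aliases a scalar value map around a threshold, exposed to the material library.
# It scales the map's per-pixel width from MF_Jam_GetPixelWidth by a smoothing pixel count, saturates it, and uses it as a SmoothStep band.
# EitherSide centers a rising edge on the threshold.
# Inner puts the band wholly below the threshold with reversed bounds, so its result is inverted.
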MaterialFunction /FMJamSystem/Widgets/Materials/MF_Jam_AntiAliasing
graph {
 properties {bExposeToLibrary=true, description=Returns an anti-aliased version of the supplied value map., libraryCategoriesText [DelMarUI]}
 nodes [
  n1 = FunctionInput(Preview=Rotator(Time=Constant() {r=0.15}) {speed=-6.2832}) {description=The value map to anti-alias., inputName=Value, inputType=FunctionInput_Scalar}
  n2 = FunctionInput() {
   bUsePreviewValueAsDefault=true
   description=How many pixels over which antialiasing occurs.
   inputName=Smoothing
   inputType=FunctionInput_Scalar
   sortPriority=2
   previewValue=<1.5 0 0 1>
  }
  n3 = Saturate(None=Multiply(A=MaterialFunctionCall!(UVs=@n1) {materialFunction=&/MF_Jam_GetPixelWidth}.PixelWidth, B=@n2) {constB=1.5})
  n4 = Multiply(A=@n3) {constB=0.5}
  n5 = FunctionInput() {
   bUsePreviewValueAsDefault=true
   description=The value around which the map is anti-aliased to.
   inputName=Threshold
   inputType=FunctionInput_Scalar
   sortPriority=1
   previewValue=<0.5 0 0 1>
  }
  FunctionOutput(None=SmoothStep(Min=Subtract(A=@n5, B=@n4), Max=Add(A=@n5, B=@n4), Value=@n1)) {outputName=EitherSide}
  FunctionOutput(None=SmoothStep(Min=@n5, Max=Subtract(A=@n5, B=@n3), Value=@n1)) {outputName=Inner, sortPriority=1}
 ]
}
